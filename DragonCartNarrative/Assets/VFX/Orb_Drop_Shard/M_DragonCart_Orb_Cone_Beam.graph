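# Additive, unlit, two-sided material for a cone-shaped light beam on Niagara mesh particles in the DragonCart orb drop shard effect.
# Particle color drives emissive through an adjustment function.
# Opacity combines a sphere mask on power-remapped V coordinates, forming a soft band along the beam, with depth fade against scene geometry.
# An axis-aligned fresnel on the tangent basis adds angle fading, scaled by base opacity and particle alpha.
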Material /DragonCartNarrative/Assets/VFX/Orb_Drop_Shard/M_DragonCart_Orb_Cone_Beam
graph {
 properties {
  bAutomaticallySetUsageInEditor=false
  bUsedWithNiagaraMeshParticles=true
  bUsedWithStaticMesh=false
  blendMode=BLEND_Additive
  previewMesh=/Game/Effects/Fort_Effects/Mesh/Shapes/SM_Cone_Wide_02
  shadingModel=MSM_Unlit
  twoSided=true
  usedShadingModels=Unlit
 }
 nodes [
  n1 = ParticleColor()
  MaterialFunctionCall(Emissive=@n1.RGB) {materialFunction=&MF_FX_EmissiveAdjustments}
  n2 = ComponentMask(None=TextureCoordinate()) {g=true}
  n3 = ScalarParameter() {defaultValue=2, parameterName=GradientUVPower}
  n4 = Constant() {r=0.5}
  n5 = DepthFade(
   Opacity=SphereMask(A=Power(Base=@n2, Exp=@n3), B=@n4, Radius=@n4, Hardness=ScalarParameter() {defaultValue=0.25, parameterName=Hardness}) {hardnessPercent=0}
   FadeDistance=ScalarParameter() {defaultValue=1000, parameterName=DepthFade}
  )
  n6 = MaterialFunctionCall(
   Axis=MaterialFunctionCall() {materialFunction=&TangentBasis}.Y V
   Exponent=ScalarParameter() {defaultValue=0.2, parameterName=AngleFadeExp}
  ) {materialFunction=&AxisAlignedFresnel}
  n7 = Power(Base=Multiply(A=@n6.Fresnel, B=@n6.Fresnel))
  Multiply(A=@n1.A, B=Multiply(A=Multiply(A=@n5, B=@n7), B=ScalarParameter() {defaultValue=0.5, parameterName=BaseOpacity}))
 ]
}
