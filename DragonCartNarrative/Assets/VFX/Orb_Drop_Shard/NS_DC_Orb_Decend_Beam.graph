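# Niagara beam effect for an orb descent, drawn between the system's position and User.BeamStartPositionWS.
# Only falling_sprites is enabled: it continuously spawns warm sprites along that line with spherical scatter and velocity along it.
# Curl noise, drag and vortex velocity centered on each particle's line point stir them as alpha and size fade in and out.
# DropBeam (sprite line), Rings (line-traveling pulsing meshes) and Ref emitters are disabled.
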NiagaraSystem /DragonCartNarrative/Assets/VFX/Orb_Drop_Shard/NS_DC_Orb_Decend_Beam
dependencies {
 usedDataInterfaces [&NiagaraDataInterfaceCamera, &NiagaraDataInterfaceCurve]
 usedDynamicInputs []
 usedModules [&SystemState]
}
emitters {
 DropBeam {
  bEnabled=false
  emitterUpdateScript [
   &EmitterState(Inactive Response=::NewEnumerator1, Life Cycle Mode=::NewEnumerator1, Loop Duration=3.5, UseLoopDelay=true)
   /SetVariables(Emitter.Cam Pos:Vector3f=&ReturnCameraVector(Camera Property=::NewEnumerator0, Coordinate Space=::Simulation)) {isSetParameters=true}
   &SpawnBurst_Instantaneous(Spawn Count=2, Spawn Group:NiagaraInt32=$Emitter.LoopCount, Use Spawn Probability=false)
  ]
  particleSpawnScript [/Niagara/Modules/Spawn/Initialization/InitializeParticle(
    Color=<rgba 0.2176 0.2843 1 0.05>
    Lifetime=4
    Material Random:NiagaraFloat=&ReturnNormalizedExecIndex()
    Mesh Scale=<50 1 1>
    Module.Write Color=true
    Module.Write Lifetime=true
    Module.Write Mass=true
    Module.Write Material Random=true
    Module.Write Position=true
    Module.Write Scale=true
    Module.Write SpriteRotation=false
    Module.Write SpriteSize=true
    Position=<0 0 0>
    Sprite Size=<500 500>
   )]
  particleUpdateScript [
   &ParticleState()
   &ScaleColor(
    Scale Alpha=&Multiply_Float(
     A=&FloatFromCurve(FloatCurve=DataInterface() {propertyValues {Curve {keys [
          {interpMode=RCIM_Cubic, leaveTangentWeight=0.0833, tangentMode=RCTM_User, time=0, value=0}
          {arriveTangentWeight=0.1389, interpMode=RCIM_Cubic, leaveTangentWeight=0.2778, tangentMode=RCTM_User, time=0.25, value=1}
          {arriveTangentWeight=0.2778, interpMode=RCIM_Cubic, leaveTangentWeight=0.1389, tangentMode=RCTM_User, time=0.75, value=1}
          {arriveTangentWeight=0.0833, interpMode=RCIM_Cubic, tangentMode=RCTM_User, time=1, value=0}
         ]}}})
     B=&Lerp_Float(
      A=0
      B=1
      Alpha=&FloatFromCurve(
       FloatCurve=DataInterface() {propertyValues {Curve {keys [
           <at 0.3 0 RCIM_Cubic>
           {arriveTangentWeight=0.3012, interpMode=RCIM_Cubic, leaveTangentWeight=0.1309, tangentMode=RCTM_User, time=0.75, value=1}
           {interpMode=RCIM_Cubic, leaveTangent=3.4262, time=1, value=1}
          ]}}}
       CurveIndex=&NormalizeDistanceRange(
        Start Position:NiagaraPosition=/Niagara/DynamicInputs/Helpers/SimulationPosition()
        End Position:NiagaraPosition=Linked() {linkedVariable {name=Emitter.Cam Pos, type=Vector3f}}
        Distance:NiagaraFloat=15000
        Invert Normalized Range:NiagaraBool=false
       )
      )
     )
    )
    ScaleA=true
    ScaleRGB=true
   ) {enabled=false}
   &ScaleColor(
    Scale Alpha=&FloatFromCurve(FloatCurve=DataInterface() {propertyValues {Curve {keys [
         {interpMode=RCIM_Cubic, leaveTangentWeight=0.0833, tangentMode=RCTM_User, time=0, value=0}
         {arriveTangentWeight=0.1389, interpMode=RCIM_Cubic, leaveTangentWeight=0.2778, tangentMode=RCTM_User, time=0.25, value=1}
         {arriveTangentWeight=0.2778, interpMode=RCIM_Cubic, leaveTangentWeight=0.1389, tangentMode=RCTM_User, time=0.75, value=1}
         {arriveTangentWeight=0.0833, interpMode=RCIM_Cubic, tangentMode=RCTM_User, time=1, value=0}
        ]}}})
    ScaleA=true
    ScaleRGB=true
   ) {module=ScaleColor001}
   /Niagara/Modules/Debug/SpriteBasedLine(Point A:NiagaraPosition=$Engine.Owner.Position, Point B:NiagaraPosition=$User.BeamStartPositionWS)
  ]
  rendererClasses [&NiagaraSpriteRendererProperties]
  renderers [{rendererClass=&NiagaraSpriteRendererProperties}]
 }
 Ref {
  bEnabled=false
  emitterUpdateScript [
   &EmitterState()
   &SpawnBurst_Instantaneous(Spawn Count=10, Use Spawn Probability=false)
  ]
  particleSpawnScript [/Niagara/Modules/Spawn/Initialization/V2/InitializeParticle(Lifetime=10, UsePositionOffset=false, Write Material Random=true)]
  particleUpdateScript [&ParticleState()]
  rendererClasses [&NiagaraSpriteRendererProperties, &NiagaraMeshRendererProperties]
  renderers [
   {rendererClass=&NiagaraSpriteRendererProperties}
   {rendererClass=&NiagaraMeshRendererProperties, rendererIndex=1}
  ]
 }
 Rings {
  bEnabled=false
  emitterUpdateScript [
   &EmitterState(Life Cycle Mode=::NewEnumerator1, Loop Duration=1, UseLoopDelay=false)
   &SpawnRate(SpawnRate=1.6667, Use Spawn Probability=false)
  ]
  particleSpawnScript [
   &InitializeParticle(
    AdjustAlpha=false
    AdjustHue=true
    AdjustSaturation=false
    AdjustValue=false
    Color=<rgba 5.0749 5.6749 20 0.02>
    Color Mode=::NewEnumerator3
    Hue Shift Range:Vector2f=<-0.05 0.1>
    Lifetime=15
    Mesh Scale=<7 7 10>
    Mesh Scale Mode=::NewEnumerator4
    UsePositionOffset=false
    Write Material Random=true
   )
   &InitialMeshOrientation(Orientation Axis=<0 0 1>, Use Orientation Vector=true, Use Rotation Vector=false)
  ]
  particleUpdateScript [
   &ParticleState()
   &self:Line_Location(
    End Pos:NiagaraPosition=$Engine.Owner.Position
    Pos On Line:NiagaraFloat=$Particles.NormalizedAge
    Start Pos:NiagaraPosition=$User.BeamStartPositionWS
   ) {module=Line_Location}
   &ScaleColor(
    Scale Alpha=&FloatFromCurve(FloatCurve=DataInterface() {propertyValues {Curve {keys [
         {interpMode=RCIM_Cubic, leaveTangent=5.0189, tangentMode=RCTM_Break, time=0, value=0}
         {
          arriveTangentWeight=0.457
          interpMode=RCIM_Cubic
          leaveTangentWeight=0.5403
          tangentMode=RCTM_Break
          tangentWeightMode=RCTWM_WeightedBoth
          time=0.5
          value=1
         }
         {arriveTangent=-3.838, interpMode=RCIM_Cubic, tangentMode=RCTM_Break, time=1, value=0}
        ]}}})
    ScaleA=true
    ScaleRGB=false
   )
   &ScaleMeshSize(Scale Factor=&VectorFromFloat(Value=&Multiply_Float(
      A=&Waveform(
       Global Phase=$Particles.NormalizedAge
       "Waveform [1]"=::NewEnumerator0
       "[1] Phase Scale"=5
       "[1] Phase"=$Particles.NormalizedAge
       "[1] Frequency"=0.75
       "[1] Amplitude Min/Max"=<1 0.25>
       "[1] Bias"=0.75
       "[2] Phase"=$Emitter.Age
       "[3] Phase"=$Emitter.Age
      )
      B=&FloatFromCurve(FloatCurve=DataInterface() {propertyValues {Curve {keys [
           {arriveTangent=3.625, interpMode=RCIM_Cubic, time=0.85, value=1}
           {arriveTangent=19.521, interpMode=RCIM_Cubic, leaveTangent=19.521, tangentMode=RCTM_User, time=1, value=2}
          ]}}})
     )))
   &SolveForcesAndVelocity(Clamp Velocity=false, Limit Acceleration=false)
  ]
  rendererClasses [&NiagaraMeshRendererProperties]
  renderers [{rendererClass=&NiagaraMeshRendererProperties, rendererIndex=0}]
 }
 falling_sprites {
  emitterUpdateScript [
   &EmitterState(Life Cycle Mode=::NewEnumerator1, Loop Duration=2, UseLoopDelay=false)
   &SpawnRate(SpawnRate=90, Use Spawn Probability=false)
  ]
  particleSpawnScript [
   &InitializeParticle(
    Color=<rgba 1 0.9453 0.6287 0.7>
    Lifetime Max=1.75
    Lifetime Min=1.4
    Lifetime Mode=::NewEnumerator1
    Mass Mode=::NewEnumerator3
    Sprite Rotation Mode=::NewEnumerator2
    Sprite Size Mode=::NewEnumerator1
    Uniform Sprite Size Max=30
    Uniform Sprite Size Min=25
    UsePositionOffset=false
    Write Material Random=true
   )
   &self:Line_Location(
    End Pos:NiagaraPosition=$User.BeamStartPositionWS
    Pos On Line:NiagaraFloat=$Particles.MaterialRandom
    Start Pos:NiagaraPosition=$Engine.Owner.Position
   ) {module=Line_Location}
   &SphereLocation(
    MaskSpawnGroup=false
    Override Seed=false
    Sphere Radius=300
    Surface Only=false
    UseNonUniformScale=false
    UseOffset=false
    UseSphereOrientation=false
   )
   &AddVelocity(Velocity=&Multiply_VectorByFloat(Vector=$Output.Line_Location.Direction, Float=350))
  ]
  particleUpdateScript [
   &ParticleState()
   &CurlNoiseForce(Noise Frequency=1, Noise Strength=5500, Pan Noise=false)
   &Drag(Drag=25, Use Linear Drag=true, Use Rotational Drag=false)
   &ScaleColor(
    Scale Alpha=&FloatFromCurve(FloatCurve=DataInterface() {propertyValues {Curve {keys [
         {interpMode=RCIM_Cubic, leaveTangent=4.5453, tangentMode=RCTM_Break, time=0, value=0}
         {interpMode=RCIM_Cubic, tangentMode=RCTM_Break, time=0.5, value=1}
         {arriveTangent=-4.8588, interpMode=RCIM_Cubic, tangentMode=RCTM_Break, time=1, value=0}
        ]}}})
    ScaleA=true
    ScaleRGB=true
   )
   &VortexForce(Vortex Force Amount=50000) {enabled=false}
   /Niagara/Modules/Update/Velocity/VortexVelocity(
    Influence Falloff Radius=49999998976
    Module.Use Influence Falloff=false
    Velocity Amount:NiagaraFloat=&RandomRangeFloat(Minimum=0, Maximum=250)
    Vortex Origin:Vector3f=&Lerp_Vector(
     A=Linked() {linkedVariable {name=Engine.Owner.Position, type=NiagaraPosition}}
     B=Linked() {linkedVariable {name=User.BeamStartPositionWS, type=NiagaraPosition}}
     Alpha=$Particles.Initial.MaterialRandom
    )
   )
   &SolveForcesAndVelocity(Clamp Velocity=false, Limit Acceleration=false)
   &ScaleSpriteSize(Scale Factor=&Vector2DFromFloat(Value=&FloatFromCurve(FloatCurve=DataInterface() {propertyValues {Curve {keys [
          {interpMode=RCIM_Cubic, tangentMode=RCTM_User, time=0, value=0}
          {interpMode=RCIM_Cubic, tangentMode=RCTM_User, time=0.5, value=1}
          <at 1 0 RCIM_Cubic>
         ]}}})))
  ]
  rendererClasses [&NiagaraSpriteRendererProperties]
  renderers [{rendererClass=&NiagaraSpriteRendererProperties}]
 }
}
summary {
 systemName=NS_DC_Orb_Decend_Beam
 emitters [
  {bEnabled=false, emitterName=DropBeam, rendererClasses [&NiagaraSpriteRendererProperties]}
  {emitterName=falling_sprites, rendererClasses [&NiagaraSpriteRendererProperties]}
  {
   bEnabled=false
   emitterName=Ref
   rendererClasses [&NiagaraSpriteRendererProperties, &NiagaraMeshRendererProperties]
  }
  {bEnabled=false, emitterName=Rings, rendererClasses [&NiagaraMeshRendererProperties]}
 ]
 userVariables [User.BeamStartPositionWS {type=NiagaraPosition, description="", defaultValue {}}]
}
