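# Niagara effect for the orb drop-shard beam: a warm orange shaft of light extending downward from an orb.
# Mesh cylinders, beam and core glow combine with flare, rays, streaks, falling motes and lights, most pinned to system orb-offset vectors each update.
# User.Height scales beam length, spawn rates and offsets, while User.SmallBeam shrinks the core and cylinder meshes.
# Camera-distance offsets adjust glow sprites, and secondary detail emitters use distance culling.
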NiagaraSystem /DragonCartNarrative/Assets/VFX/Orb_Drop_Shard/NS_Orb_DropShard_Beam
dependencies {
 usedDataInterfaces [
  &NiagaraDataInterfaceVectorCurve
  &NiagaraDataInterfaceCurve
  &NiagaraDataInterfaceColorCurve
  &NiagaraDataInterfaceCamera
  /Script/Niagara.NiagaraDataInterfaceOcclusion
  &NiagaraDataInterfaceVector2DCurve
 ]
 usedDynamicInputs [&self:ScratchDynamicInput, &TransformVector]
 usedModules [
  &self:SystemSpawnScript.SystemScriptSource.SystemScriptGraph.NiagaraNodeAssignment_0.SetVariables_#1
  &self:SystemSpawnScript.SystemScriptSource.SystemScriptGraph.NiagaraNodeAssignment_1.SetVariables_#2
  &SystemState
  &self:SystemSpawnScript.SystemScriptSource.SystemScriptGraph.NiagaraNodeAssignment_2.SetVariables_#3
  &self:SystemSpawnScript.SystemScriptSource.SystemScriptGraph.NiagaraNodeAssignment_3.SetVariables_#4
 ]
}
emitters {
 CoreBleedLight {
  emitterUpdateScript [
   &EmitterState(
    Inactive Response=::NewEnumerator1
    Life Cycle Mode=::NewEnumerator1
    Loop Behavior=::NewEnumerator1
    Loop Delay=1.7
    Loop Duration Mode=::NewEnumerator1
    UseLoopDelay=false
   )
   &SpawnBurst_Instantaneous(Use Spawn Probability=false, UseLoopCountLimit=false)
  ]
  particleSpawnScript [
   &InitializeParticle(
    Color=<rgba 3 0.6646 0.0875 1>
    Lifetime=0.4
    Mesh Scale=<4.5 4.5 6>
    Mesh Scale Mode=::NewEnumerator4
    Sprite Size Mode=::NewEnumerator3
    Uniform Sprite Size=1000
    UsePositionOffset=false
    Write Material Random=true
   )
   &InitialMeshOrientation(Mesh Orientation Mode=::NewEnumerator2, Orientation Vector=<0 0 1>, Use Rotation Vector=false)
   &SpriteFacingAndAlignment(Sprite Facing=<0 0 1>, WriteAlignment=false, WriteFacing=true)
   &DynamicMaterialParameters(
    Index 3 Param 1=2.5
    Index 3 Param 2=0.25
    Module.Index3Float0Write=true
    Module.Index3Float1Write=true
    Module.Index3Float2Write=false
    Module.Index3Float3Write=false
    Write Parameter Index 0=::NewEnumerator0
    Write Parameter Index 3=::NewEnumerator1
   )
  ]
  particleUpdateScript [
   &ParticleState(Kill Particles When Lifetime Has Elapsed=false)
   &ScaleSpriteSize(
    Uniform Curve Scale=1.5
    Uniform Curve Sprite Scale=DataInterface() {propertyValues {Curve {keys [{arriveTangent=2.288, interpMode=RCIM_Cubic, time=0, value=0}, {tangentMode=RCTM_User, time=1, value=1}]}}}
   )
   &ScaleMeshSize(Scale Factor=&VectorFromCurve(
     VectorCurve=DataInterface() {propertyValues {
       XCurve {keys [{arriveTangent=2.3661, interpMode=RCIM_Cubic, time=0, value=0}, <at 1 1>]}
       YCurve {keys [{arriveTangent=2.3661, interpMode=RCIM_Cubic, time=0, value=0}, <at 1 1>]}
       ZCurve {keys [{arriveTangent=2.3661, interpMode=RCIM_Cubic, time=0, value=0}, <at 1 1>]}
      }}
     Scale Curve=&VectorFromFloat(Value=&MakeCustomFloatFromBool(Bool=$User.SmallBeam, True Float=-1, False Float=-2))
    ))
   &SkeletalMeshLocation(
    Bone / Socket Index=4
    MaskSpawnGroup=false
    Mesh Sampling Type=::NewEnumerator5
    Sampled Position Offset=<20 0 0>
    Skeletal Mesh=$System.Skeletal Mesh
    Socket Sampling Mode=::NewEnumerator5
   ) {enabled=false}
   &ScaleColor(
    Scale Alpha=&FloatFromCurve(FloatCurve=DataInterface() {propertyValues {Curve {keys [{interpMode=RCIM_Cubic, leaveTangent=4.9886, tangentMode=RCTM_Break, time=0, value=0.1608}, <at 0.3 1>]}}})
    ScaleA=true
    ScaleRGB=true
   )
   /SetVariables(Particles.Position=&AddVectorToPosition(Position=<0 0 0>, Vector=$System.OrbWPOLocal)) {isSetParameters=true}
   &CameraOffset(Camera Offset Amount=10)
  ]
  rendererClasses [&NiagaraMeshRendererProperties, &NiagaraSpriteRendererProperties]
  renderers [
   {rendererClass=&NiagaraMeshRendererProperties, rendererIndex=0}
   {rendererClass=&NiagaraSpriteRendererProperties, rendererIndex=1}
  ]
 }
 Cylinder {
  emitterUpdateScript [
   &EmitterState(
    Inactive Response=::NewEnumerator1
    Life Cycle Mode=::NewEnumerator1
    Loop Behavior=::NewEnumerator1
    Loop Delay=1.7
    Loop Duration Mode=::NewEnumerator1
    UseLoopDelay=false
   )
   &SpawnBurst_Instantaneous(Use Spawn Probability=false, UseLoopCountLimit=false)
  ]
  particleSpawnScript [&InitializeParticle(
    Color=<rgba 1.5 0.3489 0.0445 0.4>
    Lifetime=0.4
    Mesh Scale=&MakeVector(X=&Multiply_Float(A=-100, B=$User.Height), Y=18, Z=18)
    Mesh Scale Mode=::NewEnumerator4
    UsePositionOffset=false
    Write Material Random=true
   )]
  particleUpdateScript [
   &SolveForcesAndVelocity(Clamp Velocity=false, Limit Acceleration=false)
   &ParticleState(Kill Particles When Lifetime Has Elapsed=false)
   &ScaleMeshSize(Scale Factor=&VectorFromCurve(
     VectorCurve=DataInterface() {propertyValues {
       XCurve {keys [{interpMode=RCIM_Cubic, leaveTangent=5.3095, tangentMode=RCTM_Break, time=0, value=0}, <at 0.5 1>]}
       YCurve {keys [{arriveTangent=1.2862, interpMode=RCIM_Cubic, time=0, value=0.05}, <at 1 1>]}
       ZCurve {keys [{arriveTangent=1.2862, interpMode=RCIM_Cubic, time=0, value=0.05}, <at 1 1>]}
      }}
     Scale Curve=&VectorFromFloat(Value=&MakeCustomFloatFromBool(Bool=$User.SmallBeam, True Float=0.7, False Float=1))
    ))
   /SetVariables(Particles.Position=&AddVectorToPosition(Position=<0 0 0>, Vector=$System.OrbWPO)) {isSetParameters=true, module=SetVariables_#5}
   &ScaleColor(
    Scale Alpha=&FloatFromCurve(
     FloatCurve=DataInterface() {propertyValues {Curve {keys [{interpMode=RCIM_Cubic, leaveTangent=4.9886, tangentMode=RCTM_Break, time=0, value=0.1608}, <at 0.3 1>]}}}
     Scale Curve=0.5
    )
    ScaleA=true
    ScaleRGB=true
   )
   &FN_Light_Attributes(
    Color=$Particles.Color
    Create Color Attribute=true
    Create Position Attribute=true
    Create Radius Attribute=true
    Position=&AddVectorToPosition(Position=$Particles.Position, Vector=<0 0 0>)
    Radius=5000
    Scale Color=3
   )
  ]
  rendererClasses [&NiagaraMeshRendererProperties]
  renderers [{rendererClass=&NiagaraMeshRendererProperties, rendererIndex=0}]
 }
 Cylinder_Low {
  emitterUpdateScript [
   &EmitterState(
    Inactive Response=::NewEnumerator1
    Life Cycle Mode=::NewEnumerator1
    Loop Behavior=::NewEnumerator1
    Loop Delay=1.7
    Loop Duration Mode=::NewEnumerator1
    UseLoopDelay=false
   )
   &SpawnBurst_Instantaneous(Use Spawn Probability=false, UseLoopCountLimit=false)
  ]
  particleSpawnScript [&InitializeParticle(
    Color=<rgba 1.5 0.3489 0.0445 0.4>
    Lifetime=0.4
    Mesh Scale=&MakeVector(X=&Multiply_Float(A=-25, B=$User.Height), Y=16, Z=16)
    Mesh Scale Mode=::NewEnumerator4
    UsePositionOffset=false
    Write Material Random=true
   )]
  particleUpdateScript [
   &SolveForcesAndVelocity(Clamp Velocity=false, Limit Acceleration=false)
   &ParticleState(Kill Particles When Lifetime Has Elapsed=false)
   &ScaleMeshSize(Scale Factor=&VectorFromCurve(
     VectorCurve=DataInterface() {propertyValues {
       XCurve {keys [{interpMode=RCIM_Cubic, leaveTangent=5.3095, tangentMode=RCTM_Break, time=0, value=0}, <at 0.5 1>]}
       YCurve {keys [{arriveTangent=1.2862, interpMode=RCIM_Cubic, time=0, value=0.05}, <at 1 1>]}
       ZCurve {keys [{arriveTangent=1.2862, interpMode=RCIM_Cubic, time=0, value=0.05}, <at 1 1>]}
      }}
     Scale Curve=&VectorFromFloat(Value=&MakeCustomFloatFromBool(Bool=$User.SmallBeam, True Float=0.7, False Float=1))
    ))
   /SetVariables(Particles.Position=&AddVectorToPosition(Position=<0 0 -3700>, Vector=$System.OrbWPO)) {isSetParameters=true, module=SetVariables_#5}
   &ScaleColor(
    Scale Alpha=&FloatFromCurve(
     FloatCurve=DataInterface() {propertyValues {Curve {keys [{interpMode=RCIM_Cubic, leaveTangent=4.9886, tangentMode=RCTM_Break, time=0, value=0.1608}, <at 0.3 1>]}}}
     Scale Curve=0.7
    )
    ScaleA=true
    ScaleRGB=true
   )
   &FN_Light_Attributes(
    Color=$Particles.Color
    Create Color Attribute=true
    Create Position Attribute=true
    Create Radius Attribute=true
    Position=&AddVectorToPosition(Position=$Particles.Position, Vector=<0 0 0>)
    Radius=5000
    Scale Color=3
   )
  ]
  rendererClasses [&NiagaraMeshRendererProperties]
  renderers [{rendererClass=&NiagaraMeshRendererProperties, rendererIndex=0}]
 }
 DarkSPline001 {
  emitterUpdateScript [
   &EmitterState(
    Enable Distance Culling=true
    Life Cycle Mode=::NewEnumerator1
    Loop Delay=0.8
    Loop Duration=3
    MaxDistance=8000
    MinDistance=4000
    Module.bUseMaxDistance=true
    Module.bUseMinDistance=true
    Scalability Mode=::NewEnumerator1
    UseLoopDelay=true
   )
   &SpawnBurst_Instantaneous(Spawn Count=2, Spawn Time=0.2, Use Spawn Probability=false, UseLoopCountLimit=false)
   &SpawnRate(SpawnRate=&Multiply_Float(A=4, B=$User.Height), Use Spawn Probability=false)
  ]
  particleSpawnScript [
   &InitializeParticle(
    Color=<rgba 5 1.1728 0 0.75>
    Lifetime Max=4
    Lifetime Min=3
    Lifetime Mode=::NewEnumerator1
    Mass Max=0.2
    Mass Min=0
    Mass Mode=::NewEnumerator3
    Mesh Scale Max=<15 15 10>
    Mesh Scale Min=<-15 -15 5>
    Mesh Scale Mode=::NewEnumerator2
    Position Offset=<0 0 -500>
    Sprite Rotation Mode=::NewEnumerator2
    UsePositionOffset=false
    Write Material Random=true
   )
   &InitialMeshOrientation(
    Mesh Orientation Mode=::NewEnumerator2
    Rotation=&RandomRangeVector(Minimum=<0 0 0>, Maximum=<0 0 1>)
    Use Rotation Vector=true
   )
   &AddVelocity(
    Cone Angle=1
    Cone Axis=<0 0 -1>
    Module.UseConeFalloff=false
    Module.UseVelDistribution=false
    Override Seed=false
    Velocity Mode=::NewEnumerator2
    Velocity Speed=&RandomRangeFloat(Minimum=1, Maximum=2)
   )
   &ShapeLocation(MaskSpawnGroup=false, Offset Mode=::NewEnumerator0, Override Seed=false, UseOwnerScale=true) {enabled=false}
   &ShapeLocation(
    Cylinder Height=&Multiply_Float(A=3800, B=$User.Height)
    Cylinder Height Midpoint=1
    Cylinder Radius=60
    MaskSpawnGroup=false
    Override Seed=false
    Shape Primitive=::NewEnumerator1
    Surface Only=false
    UseOwnerScale=true
   ) {module=ShapeLocation001}
  ]
  particleUpdateScript [
   &ParticleState()
   &ScaleMeshSize(Scale Factor=&VectorFromCurve(
     VectorCurve=DataInterface() {propertyValues {
       XCurve {keys [
         {arriveTangent=1.9584, interpMode=RCIM_Cubic, leaveTangent=1.9584, tangentMode=RCTM_User, time=0, value=0.2}
         {arriveTangent=1.1179, interpMode=RCIM_Cubic, leaveTangent=1.1179, tangentMode=RCTM_User, time=0.3, value=1.5}
         {arriveTangent=0.0224, interpMode=RCIM_Cubic, leaveTangent=0.0224, tangentMode=RCTM_User, time=1, value=2}
        ]}
       YCurve {keys [
         {arriveTangent=1.9584, interpMode=RCIM_Cubic, leaveTangent=1.9584, tangentMode=RCTM_User, time=0, value=0.2}
         {arriveTangent=1.1179, interpMode=RCIM_Cubic, leaveTangent=1.1179, tangentMode=RCTM_User, time=0.3, value=1.5}
         {arriveTangent=0.0224, interpMode=RCIM_Cubic, leaveTangent=0.0224, tangentMode=RCTM_User, time=1, value=2}
        ]}
       ZCurve {keys [
         {arriveTangent=1.9584, interpMode=RCIM_Cubic, leaveTangent=1.9584, tangentMode=RCTM_User, time=0, value=0.2}
         {arriveTangent=1.1179, interpMode=RCIM_Cubic, leaveTangent=1.1179, tangentMode=RCTM_User, time=0.3, value=0.75}
         {arriveTangent=0.0224, interpMode=RCIM_Cubic, leaveTangent=0.0224, tangentMode=RCTM_User, time=1, value=1}
        ]}
      }}
     Scale Curve=&Multiply_VectorByFloat(Vector=<20 20 16>, Float=0.07)
    ))
   &Drag(Drag=&RandomRangeFloat(Minimum=0.25, Maximum=0.35), Use Linear Drag=true, Use Rotational Drag=false)
   &ScaleColor(
    Scale Alpha=&FloatFromCurve(
     FloatCurve=DataInterface() {propertyValues {Curve {keys [
         {arriveTangent=6.3495, interpMode=RCIM_Cubic, leaveTangent=6.3495, tangentMode=RCTM_User, time=0, value=0}
         {interpMode=RCIM_Cubic, tangentMode=RCTM_User, time=0.5, value=1}
         {tangentMode=RCTM_User, time=1, value=0}
        ]}}}
     Scale Curve=0.7
    )
    ScaleA=true
    ScaleRGB=true
   )
   &CameraOffset(Camera Offset Amount=&Multiply_Float(
     A=/Game/Effects/Niagara/Modules/Math/FN_CameraDistance(
      0Dist:NiagaraFloat=0
      1Dist:NiagaraFloat=100
      Camera Query:NiagaraDataInterfaceCamera=DataInterface() {propertyValues {bRequireCurrentFrameData=false}}
      Position:NiagaraPosition=$Particles.Position
     )
     B=300
    ))
   &SolveForcesAndVelocity(Clamp Velocity=false, Limit Acceleration=false)
   &DynamicMaterialParameters(
    Index 0 Param 1=0.25
    Index 0 Param 3=1
    Module.Index0Float0Write=true
    Module.Index0Float1Write=true
    Module.Index0Float2Write=true
    Module.Index0Float3Write=true
    Module.Index1Float0Write=true
    Module.Index1Float1Write=true
    Module.Index1Float2Write=false
    Module.Index1Float3Write=false
    Write Parameter Index 1=::NewEnumerator1
   )
   &ScaleColor(
    Scale Alpha=&MakeFloatFromLinearColor(
     LinearColor=&LerpLinearColors(
      StartColor=<rgba 0 0 0 1>
      EndColor=<rgba 1 1 1 1>
      LerpFactor=&FN_LOD_Distance_Remap(Start Distance=200, End Distance=5000, Minimum Output=0, Maximum Output=1, Clamp Output=true)
     )
     Channel:ENiagaraFloatFromLinearColorOptions=::NewEnumerator6
    )
    ScaleA=true
    ScaleRGB=true
   ) {enabled=false, module=ScaleColor001}
  ]
  rendererClasses [&NiagaraMeshRendererProperties]
  renderers [{rendererClass=&NiagaraMeshRendererProperties, rendererIndex=0}]
 }
 EnergyLeak001 {
  emitterUpdateScript [
   &EmitterState(Life Cycle Mode=::NewEnumerator1, Loop Delay=0.4, Loop Duration=2, UseLoopDelay=false)
   &SpawnBurst_Instantaneous(Use Spawn Probability=false, UseLoopCountLimit=false)
  ]
  particleSpawnScript [
   &InitializeParticle(
    Color=<rgba 0.9271 0.1288 0 1>
    Lifetime=4
    Mesh Scale Mode=::NewEnumerator3
    Mesh Uniform Scale=&Multiply_Float(A=10, B=1)
    Sprite Rotation Mode=::NewEnumerator2
    Sprite Size Mode=::NewEnumerator3
    Sprite UV Mode=::NewEnumerator3
    Uniform Sprite Size=&Multiply_Float(A=900, B=1)
    UsePositionOffset=false
    Write Material Random=true
   )
   &InitialMeshOrientation(Mesh Axis To Orient=<0 0 1>, Mesh Orientation Mode=::NewEnumerator2, Use Rotation Vector=false)
  ]
  particleUpdateScript [
   &ParticleState()
   &ScaleSpriteSize(
    Uniform Curve Scale=2
    Uniform Curve Sprite Scale=DataInterface() {propertyValues {Curve {keys [
        {interpMode=RCIM_Cubic, tangentMode=RCTM_Break, time=0, value=0.5}
        {arriveTangent=-0.3016, interpMode=RCIM_Cubic, tangentMode=RCTM_Break, time=0.75, value=1}
       ]}}}
   )
   &Timeline(
    Play Rate=&FloatFromCurve(FloatCurve=DataInterface() {propertyValues {Curve {keys [<at 0 2>,<at 1 1.3>]}}})
    Timeline End=1000
   )
   &CameraOffset(Camera Offset Amount=&Multiply_Float(
     A=-500
     B=&FloatFromCurve(
      FloatCurve=DataInterface() {propertyValues {Curve {keys [<at 0 1>,<at 1 0>]}}}
      CurveIndex=&FN_CameraDistance(
       0Dist:NiagaraFloat=500
       1Dist:NiagaraFloat=2000
       Camera Query:NiagaraDataInterfaceCamera=DataInterface()
       Position:NiagaraPosition=$Particles.Position
      )
     )
    )) {enabled=false}
   &ScaleColor(
    Scale Alpha=&FloatFromCurve(
     FloatCurve=DataInterface() {propertyValues {Curve {keys [
         {arriveTangent=9.0725, interpMode=RCIM_Cubic, leaveTangent=9.0725, tangentMode=RCTM_User, time=0, value=0}
         {arriveTangent=0.1216, interpMode=RCIM_Cubic, leaveTangent=0.1216, tangentMode=RCTM_User, time=0.2, value=0.9847}
         {arriveTangent=-3.3987, interpMode=RCIM_Cubic, leaveTangent=-3.3987, tangentMode=RCTM_User, time=1, value=0}
        ]}}}
     Scale Curve=1.2
    )
    ScaleA=true
    ScaleRGB=true
   ) {module=ScaleColor002}
   &DynamicMaterialParameters(
    Index 0 Param 1=&Multiply_Float(A=$Particles.Timeline.Time, B=0.5)
    Module.Index0Float0Write=true
    Module.Index0Float1Write=true
    Module.Index0Float2Write=true
    Module.Index0Float3Write=true
   )
   /SetVariables(Particles.Position=&AddVectorToPosition(Position=<0 0 0>, Vector=$System.OrbWPOLocal)) {isSetParameters=true}
   &CameraOffset(Camera Offset Amount=&Multiply_Float(
     A=&FN_CameraDistance(
      0Dist:NiagaraFloat=0
      1Dist:NiagaraFloat=800
      Camera Query:NiagaraDataInterfaceCamera=DataInterface() {propertyValues {bRequireCurrentFrameData=false}}
      Position:NiagaraPosition=$Particles.Position
     )
     B=1000
    )) {module=CameraOffset001}
  ]
  rendererClasses [&NiagaraSpriteRendererProperties]
  renderers [{rendererClass=&NiagaraSpriteRendererProperties}]
 }
 Impact_Flare {
  simTarget=GPUComputeSim
  emitterUpdateScript [
   &EmitterState(Life Cycle Mode=::NewEnumerator1, Loop Duration=1, UseLoopDelay=false)
   &SpawnBurst_Instantaneous(Use Spawn Probability=false, UseLoopCountLimit=false)
  ]
  particleSpawnScript [
   &InitializeParticle(Lifetime=6, Sprite Size Mode=::NewEnumerator3, Uniform Sprite Size=3000, UsePositionOffset=false, Write Material Random=true)
   &DynamicMaterialParameters(
    Index 0 Param 1=1
    Index 0 Param 2=1
    Index 0 Param 3=1
    Index 0 Param 4=1
    Module.Index0Float0Write=true
    Module.Index0Float1Write=true
    Module.Index0Float2Write=true
    Module.Index0Float3Write=true
   )
  ]
  particleUpdateScript [
   &ParticleState()
   &self:MaskInScreenSpace_02() {enabled=false, module=MaskInScreenSpace_02}
   &self:ScratchModule_04(NumberOfRings=3, OcclusionSize=300, SamplesPerRing=8) {enabled=false, module=ScratchModule_04}
   &CameraOffset(Camera Offset Amount=&Multiply_Float(
     A=&FN_CameraDistance(
      0Dist:NiagaraFloat=0
      1Dist:NiagaraFloat=8000
      Camera Query:NiagaraDataInterfaceCamera=DataInterface() {propertyValues {bRequireCurrentFrameData=false}}
      Position:NiagaraPosition=$Particles.Position
     )
     B=1000
    ))
   &ScaleColor(
    Scale Alpha=&FloatFromCurve(
     FloatCurve=DataInterface() {propertyValues {Curve {keys [
         {arriveTangent=-0.1319, interpMode=RCIM_Cubic, leaveTangent=-0.1319, tangentMode=RCTM_User, time=0, value=0}
         <at 0.5 1 RCIM_Cubic>
         <at 1 0 RCIM_Cubic>
        ]}}}
     Scale Curve=0.4
    )
    Scale RGB=&MakeVectorFromLinearColorRGB(Color=&ColorFromCurve(ColorCurve=DataInterface() {propertyValues {
        AlphaCurve {keys []}
        BlueCurve {keys [<at 0 1.1082 RCIM_Cubic>,<at 0.3877 0.0275>,<at 1 0.025 RCIM_Cubic>]}
        GreenCurve {keys [<at 0 1.7837 RCIM_Cubic>,<at 0.3877 0.6197>,<at 1 0 RCIM_Cubic>]}
        RedCurve {keys [<at 0 2 RCIM_Cubic>,<at 0.3877 1>,<at 1 0.1878 RCIM_Cubic>]}
       }}))
    ScaleA=true
    ScaleRGB=true
   )
   &ScaleColor(ScaleA=true, ScaleRGB=true) {module=ScaleColor001}
   /SetVariables(Particles.Position=&AddVectorToPosition(Position=<0 0 0>, Vector=$System.OrbWPOLocal)) {isSetParameters=true}
   &FN_Sprite_Attributes(
    Color=&Multiply_LinearColorByFloat(Linear Color=$Particles.Color, Float=2)
    Create Color Attribute=true
    Create SpriteSize Attribute=true
    SpriteSize=&Multiply_Vector2D(A=$Particles.SpriteSize, B=<1 0.4232>)
    Write to Module Namespace=true
   )
  ]
  rendererClasses [&NiagaraSpriteRendererProperties]
  renderers [
   {rendererClass=&NiagaraSpriteRendererProperties}
   {rendererClass=&NiagaraSpriteRendererProperties, rendererIndex=1}
  ]
 }
 Light {
  emitterUpdateScript [
   &EmitterState(
    Inactive Response=::NewEnumerator1
    Life Cycle Mode=::NewEnumerator1
    Loop Behavior=::NewEnumerator1
    Loop Delay=1.7
    Loop Duration Mode=::NewEnumerator1
    UseLoopDelay=false
   )
   &SpawnRate(Use Spawn Probability=false) {enabled=false}
   &SpawnBurst_Instantaneous(Use Spawn Probability=false, UseLoopCountLimit=false)
  ]
  particleSpawnScript [&InitializeParticle(Color=<rgba 1.5 0.226 0 1>, Position Offset=<0 0 -3000>, UsePositionOffset=false, Write Material Random=true)]
  particleUpdateScript [
   &SolveForcesAndVelocity(Clamp Velocity=false, Limit Acceleration=false)
   &ParticleState(Kill Particles When Lifetime Has Elapsed=false)
   &ScaleMeshSize(Scale Factor=&VectorFromCurve(
     VectorCurve=DataInterface() {propertyValues {
       XCurve {keys [{interpMode=RCIM_Cubic, leaveTangent=5.3095, tangentMode=RCTM_Break, time=0, value=0}, <at 0.5 1>]}
       YCurve {keys [{arriveTangent=1.2862, interpMode=RCIM_Cubic, time=0, value=0.05}, <at 1 1>]}
       ZCurve {keys [{arriveTangent=1.2862, interpMode=RCIM_Cubic, time=0, value=0.05}, <at 1 1>]}
      }}
     Scale Curve=&VectorFromFloat(Value=&MakeCustomFloatFromBool(Bool=$User.SmallBeam, True Float=0.7, False Float=1))
    )) {enabled=false}
   &SkeletalMeshLocation(
    Bone / Socket Index=4
    MaskSpawnGroup=false
    Mesh Sampling Type=::NewEnumerator5
    Sampled Position Offset=<-5 0 0>
    Skeletal Mesh=$System.Skeletal Mesh
    Socket Sampling Mode=::NewEnumerator5
   ) {enabled=false}
   &AlignSpriteToMeshOrientation(SetAlignment=true, SetFacing=true) {enabled=false}
   /SetVariables(Particles.Position=&AddVectorToPosition(Position=<0 0 -300>, Vector=$System.OrbWPOLocal)) {isSetParameters=true, module=SetVariables_#6}
   &ScaleColor(
    Scale Alpha=&FloatFromCurve(FloatCurve=DataInterface() {propertyValues {Curve {keys [{interpMode=RCIM_Cubic, leaveTangent=4.9886, tangentMode=RCTM_Break, time=0, value=0.1608}, <at 0.3 1>]}}})
    ScaleA=true
    ScaleRGB=true
   ) {enabled=false}
   &FN_Light_Attributes(
    Color=$Particles.Color
    Create Color Attribute=true
    Create Radius Attribute=true
    Radius=2000
    Scale Color=&FloatFromCurve(
     FloatCurve=DataInterface() {propertyValues {Curve {
        defaultValue=1
        keys [{interpMode=RCIM_Cubic, tangentMode=RCTM_User, time=0, value=0}, {interpMode=RCIM_Cubic, tangentMode=RCTM_User, time=0.5, value=1}]
       }}}
     Scale Curve=16
    )
   )
  ]
  rendererClasses [&NiagaraLightRendererProperties]
  renderers [{rendererClass=&NiagaraLightRendererProperties, rendererIndex=0}]
 }
 Light_End {
  emitterUpdateScript [
   &EmitterState(
    Inactive Response=::NewEnumerator1
    Life Cycle Mode=::NewEnumerator1
    Loop Behavior=::NewEnumerator1
    Loop Delay=1.7
    Loop Duration Mode=::NewEnumerator1
    UseLoopDelay=false
   )
   &SpawnRate(Use Spawn Probability=false) {enabled=false}
   &SpawnBurst_Instantaneous(Use Spawn Probability=false, UseLoopCountLimit=false)
  ]
  particleSpawnScript [&InitializeParticle(Color=<rgba 1 0.1507 0 1>, Position Offset=<0 0 -5100>, UsePositionOffset=false, Write Material Random=true)]
  particleUpdateScript [
   &SolveForcesAndVelocity(Clamp Velocity=false, Limit Acceleration=false)
   &ParticleState(Kill Particles When Lifetime Has Elapsed=false)
   &ScaleMeshSize(Scale Factor=&VectorFromCurve(
     VectorCurve=DataInterface() {propertyValues {
       XCurve {keys [{interpMode=RCIM_Cubic, leaveTangent=5.3095, tangentMode=RCTM_Break, time=0, value=0}, <at 0.5 1>]}
       YCurve {keys [{arriveTangent=1.2862, interpMode=RCIM_Cubic, time=0, value=0.05}, <at 1 1>]}
       ZCurve {keys [{arriveTangent=1.2862, interpMode=RCIM_Cubic, time=0, value=0.05}, <at 1 1>]}
      }}
     Scale Curve=&VectorFromFloat(Value=&MakeCustomFloatFromBool(Bool=$User.SmallBeam, True Float=0.7, False Float=1))
    )) {enabled=false}
   &SkeletalMeshLocation(
    Bone / Socket Index=4
    MaskSpawnGroup=false
    Mesh Sampling Type=::NewEnumerator5
    Sampled Position Offset=<-5 0 0>
    Skeletal Mesh=$System.Skeletal Mesh
    Socket Sampling Mode=::NewEnumerator5
   ) {enabled=false}
   &AlignSpriteToMeshOrientation(SetAlignment=true, SetFacing=true) {enabled=false}
   /SetVariables(Particles.Position=&AddVectorToPosition(
     Position=&MakeVector(X=0, Y=0, Z=&Multiply_Float(A=-5200, B=$User.Height))
     Vector=$System.OrbWPOLocal
    )) {isSetParameters=true, module=SetVariables_#6}
   &ScaleColor(
    Scale Alpha=&FloatFromCurve(FloatCurve=DataInterface() {propertyValues {Curve {keys [{interpMode=RCIM_Cubic, leaveTangent=4.9886, tangentMode=RCTM_Break, time=0, value=0.1608}, <at 0.3 1>]}}})
    ScaleA=true
    ScaleRGB=true
   ) {enabled=false}
   &FN_Light_Attributes(
    Color=$Particles.Color
    Create Color Attribute=true
    Create Radius Attribute=true
    Radius=600
    Scale Color=&FloatFromCurve(
     FloatCurve=DataInterface() {propertyValues {Curve {
        defaultValue=1
        keys [{interpMode=RCIM_Cubic, tangentMode=RCTM_User, time=0, value=0}, {interpMode=RCIM_Cubic, tangentMode=RCTM_User, time=0.5, value=1}]
       }}}
     Scale Curve=16
    )
   )
  ]
  rendererClasses [&NiagaraLightRendererProperties]
  renderers [{rendererClass=&NiagaraLightRendererProperties, rendererIndex=0}]
 }
 PickupBeam {
  emitterUpdateScript [
   &EmitterState(Inactive Response=::NewEnumerator1, Life Cycle Mode=::NewEnumerator1, Loop Duration=0.7, UseLoopDelay=false)
   &SpawnBurst_Instantaneous(Use Spawn Probability=false, UseLoopCountLimit=false)
  ]
  particleSpawnScript [
   &InitializeParticle(
    Color=<rgba 1 0.4048 0.1927 1>
    Lifetime=2
    Mesh Scale Mode=::NewEnumerator3
    Mesh Uniform Scale=&Multiply_Float(A=22, B=$User.Height)
    UsePositionOffset=false
    Write Material Random=true
   )
   &Color(Color=&MakeLinearColorFromVectorAndFloat("Vector (RGB)"=<0.1206 0.3079 2>, "Float (Alpha)"=&ClampFloat(Float=1))) {enabled=false}
   /Niagara/Modules/CascadeConversion/CascadeConversion_InitialMeshRotation(Rotation:Vector3f=<0 0.75 0>)
  ]
  particleUpdateScript [
   &ParticleState()
   /SetVariables(Particles.Position=&AddVectorToPosition(Position=<0 0 0>, Vector=$System.OrbWPOLocal)) {isSetParameters=true}
   &ScaleMeshSize(Scale Factor=&VectorFromCurve(VectorCurve=DataInterface() {propertyValues {XCurve {keys [<at 0 0.5>,<at 0.1 1>]}, YCurve {keys [<at 0 1>,<at 0.1 1>]}, ZCurve {keys [<at 0 0.5>,<at 0.1 1>]}}})) {enabled=false}
   &ScaleColor(
    Scale Alpha=&FloatFromCurve(
     FloatCurve=DataInterface() {propertyValues {Curve {
        defaultValue=0
        keys [
         {interpMode=RCIM_Cubic, tangentMode=RCTM_User, time=0, value=0.5}
         {interpMode=RCIM_Cubic, tangentMode=RCTM_User, time=0.5, value=0.9}
         {tangentMode=RCTM_User, time=1, value=0.5}
        ]
       }}}
     Scale Curve=&FloatFromCurve(
      FloatCurve=DataInterface() {propertyValues {Curve {
         defaultValue=1
         keys [
          {interpMode=RCIM_Cubic, tangentMode=RCTM_User, time=0, value=0}
          {interpMode=RCIM_Cubic, tangentMode=RCTM_User, time=0.5, value=1}
          {tangentMode=RCTM_User, time=1, value=0}
         ]
        }}}
      Scale Curve=1.6
     )
    )
    ScaleA=true
    ScaleRGB=true
   )
   &DynamicMaterialParameters(
    Index 0 Param 1=&FloatFromCurve(
     FloatCurve=DataInterface() {propertyValues {Curve {keys [
         <at 0 25 RCIM_Cubic>
         {arriveTangent=404.6076, interpMode=RCIM_Cubic, leaveTangent=404.6076, time=0.0919, value=54.1035}
         {arriveTangent=289.8687, interpMode=RCIM_Cubic, leaveTangent=289.8687, time=0.2509, value=126.5119}
         {interpMode=RCIM_Cubic, leaveTangent=187.9053, time=0.3417, value=126.5119}
        ]}}}
     Scale Curve=7
    )
    Index 0 Param 2=1
    Index 0 Param 3=1
    Index 0 Param 4=1
    Module.Index0Float0Write=true
    Module.Index0Float1Write=true
    Module.Index0Float2Write=true
    Module.Index0Float3Write=true
   )
   &CameraOffset(Camera Offset Amount=20)
  ]
  rendererClasses [&NiagaraMeshRendererProperties]
  renderers [{rendererClass=&NiagaraMeshRendererProperties, rendererIndex=0}]
 }
 PlasmaWave {rendererClasses [], renderers []}
 Rays {
  simTarget=GPUComputeSim
  emitterUpdateScript [
   &EmitterState(
    Enable Distance Culling=true
    Inactive Response=::NewEnumerator1
    Life Cycle Mode=::NewEnumerator1
    Loop Duration=0.25
    MaxDistance=8000
    MinDistance=4000
    Module.bUseMaxDistance=true
    Module.bUseMinDistance=true
    Scalability Mode=::NewEnumerator1
    UseLoopDelay=true
   )
   &SpawnRate(SpawnRate=10, Use Spawn Probability=false) {enabled=false}
   &SpawnBurst_Instantaneous(
    Spawn Count=&RandomRangeInt(Minimum=1, Maximum=2)
    Use Spawn Probability=false
    UseLoopCountLimit=false
   )
  ]
  particleSpawnScript [
   &InitializeParticle(
    Color=<rgba 500 118.9685 0 0.2>
    Lifetime=&RandomRangeFloat(Minimum=5, Maximum=7)
    Sprite Size=&RandomRangeVector2D(Minimum=<160 120>, Maximum=<180 400>)
    Sprite Size Mode=::NewEnumerator4
    Sprite UV Mode=::NewEnumerator4
    Sprite UV Scale:Vector2f=<1 -1>
    UsePositionOffset=false
    Write Material Random=true
   )
   &ShapeLocation(
    Cylinder Height=0.1
    Cylinder Radius=260
    MaskSpawnGroup=false
    Offset Mode=::NewEnumerator0
    Override Seed=false
    Rotation Axis=<0 0 1>
    Rotation Mode=::NewEnumerator1
    Shape Primitive=::NewEnumerator1
    Surface Only=false
    UseOwnerScale=true
   ) {enabled=false}
   &AddVelocity(Velocity=&RandomRangeVector(Minimum=<0 0 -50>, Maximum=<0 0 -5000>)) {enabled=false}
   &DynamicMaterialParameters(
    Index 0 Param 1=&RandomRangeFloat(Minimum=0.8, Maximum=2)
    Index 0 Param 2=&RandomRangeFloat(Minimum=0.5, Maximum=1.5)
    Index 0 Param 3=&RandomRangeFloat(Minimum=0.6, Maximum=1.6)
    Index 0 Param 4=&RandomRangeFloat(Minimum=0.5, Maximum=1.5)
    Module.Index0Float0Write=true
    Module.Index0Float1Write=true
    Module.Index0Float2Write=true
    Module.Index0Float3Write=true
   )
  ]
  particleUpdateScript [
   &ParticleState()
   &ScaleSpriteSize(
    Scale Factor=&MakeVector2D(
     X=1
     Y=&FloatFromCurve(
      FloatCurve=DataInterface() {propertyValues {Curve {keys [
          {arriveTangent=4.1785, interpMode=RCIM_Cubic, leaveTangent=4.1785, tangentMode=RCTM_User, time=0, value=0}
          {arriveTangent=1.1179, interpMode=RCIM_Cubic, leaveTangent=1.1179, tangentMode=RCTM_User, time=0.2042, value=0.6228}
          {arriveTangent=0.0224, interpMode=RCIM_Cubic, leaveTangent=0.0224, tangentMode=RCTM_User, time=1, value=1}
         ]}}}
      Scale Curve=&Multiply_Float(A=&RandomRangeFloat(Minimum=20, Maximum=25), B=$User.Height)
     )
    )
    Scale Sprite Size Mode=::NewEnumerator2
   )
   /SetVariables(Particles.Position=&AddVectorToPosition(Position=&RandomRangeVector(Minimum=<-270 -270 0>, Maximum=<270 270 0>), Vector=$System.OrbWPOLocal)) {isSetParameters=true}
   &SolveForcesAndVelocity(Clamp Velocity=false, Limit Acceleration=false)
   &CameraOffset(Camera Offset Amount=25)
   &ScaleColor(
    Linear Color Curve=DataInterface() {propertyValues {
      AlphaCurve {keys [<at 0 0>,<at 0.1796 0.1>,<at 1 0>]}
      BlueCurve {keys [<at 0 1>,<at 1 0.1483>]}
      GreenCurve {keys [<at 0 1>,<at 1 0.2501>]}
      RedCurve {keys [<at 0 1>,<at 1 0.3229>]}
     }}
    Scale Mode=::NewEnumerator2
   )
  ]
  rendererClasses [&NiagaraSpriteRendererProperties]
  renderers [{rendererClass=&NiagaraSpriteRendererProperties}]
 }
 UpwardStreaks {
  emitterUpdateScript [
   &EmitterState(
    Enable Distance Culling=true
    Life Cycle Mode=::NewEnumerator1
    Loop Delay=0.5
    Loop Duration=7
    MaxDistance=8000
    MinDistance=4000
    Module.bUseMaxDistance=true
    Module.bUseMinDistance=true
    Scalability Mode=::NewEnumerator1
    UseLoopDelay=true
   )
   &SpawnRate(SpawnRate=&RandomRangeFloat(Minimum=1, Maximum=3), Use Spawn Probability=false)
   &SpawnBurst_Instantaneous(Loop Count Limit=0, Use Spawn Probability=false, UseLoopCountLimit=false) {enabled=false}
  ]
  particleSpawnScript [
   &InitializeParticle(
    Color=<rgba 1.2 0.9195 0.6813 0.75>
    Lifetime=&RandomRangeFloat(Minimum=3, Maximum=5)
    Sprite Size=&RandomRangeVector2D(Minimum=<50 3500>, Maximum=<80 5500>)
    Sprite Size Mode=::NewEnumerator4
    UsePositionOffset=false
    Write Material Random=true
   )
   &DynamicMaterialParameters(
    Index 0 Param 1=1
    Index 0 Param 2=4
    Index 0 Param 3=1
    Index 0 Param 4=1
    Module.Index0Float0Write=true
    Module.Index0Float1Write=true
    Module.Index0Float2Write=true
    Module.Index0Float3Write=true
   )
   &ShapeLocation(
    Cylinder Height=0
    Cylinder Height Midpoint=1000
    Cylinder Radius=50
    MaskSpawnGroup=false
    Non Uniform Scale=<4 1 2000>
    Offset=&MakeVector(X=0, Y=0, Z=&Multiply_Float(A=-1500, B=$User.Height))
    Offset Coordinate Space=::Local
    Offset Mode=::NewEnumerator0
    Override Seed=false
    Shape Origin=$Engine.Emitter.SimulationPosition
    Shape Primitive=::NewEnumerator1
    Surface Only=false
    UseOwnerScale=true
   ) {enabled=false}
   &AddVelocity(Velocity=&RandomRangeVector(
     Minimum=&Multiply_VectorByFloat(Vector=<1 1 -250>, Float=$User.Height)
     Maximum=&Multiply_VectorByFloat(Vector=<1 1 -400>, Float=$User.Height)
    ))
  ]
  particleUpdateScript [
   &ParticleState()
   &SolveForcesAndVelocity(Clamp Velocity=false, Limit Acceleration=false)
   /SetVariables(Particles.Position=&AddVectorToPosition(
     Position=&RandomRangeVector(
      Minimum=&MakeVector(X=-150, Y=-150, Z=&Multiply_Float(A=-1000, B=$User.Height))
      Maximum=&MakeVector(X=150, Y=150, Z=&Multiply_Float(A=-3000, B=$User.Height))
     )
     Vector=$System.OrbWPOLocal
    )) {isSetParameters=true}
   &CameraOffset(Camera Offset Amount=27.1877)
   &self:ColorSet_01(
    Linear Color:LinearColor=<rgba 1 0.8522 0.3265 0.5>
    Random=0.5
    ScaleCurve:NiagaraFloat=&FloatFromCurve(
     FloatCurve=DataInterface() {propertyValues {Curve {keys [
         {arriveTangent=0.0641, interpMode=RCIM_Cubic, leaveTangent=0.0641, tangentMode=RCTM_User, time=0, value=0}
         {arriveTangent=-2.5249, interpMode=RCIM_Cubic, leaveTangent=-2.5249, tangentMode=RCTM_User, time=0.4346, value=1}
         {arriveTangent=-0.0405, interpMode=RCIM_Cubic, leaveTangent=-0.0405, tangentMode=RCTM_User, time=1, value=0}
        ]}}}
     Scale Curve=0.7
    )
   ) {enabled=false, module=ColorSet_01}
   &ScaleSpriteSize(
    Scale Factor=&Vector2DFromCurve(
     Vector2Curve=DataInterface() {propertyValues {
       XCurve {keys [<at 0 2.15 RCIM_Cubic>,<at 1 0 RCIM_Cubic>]}
       YCurve {keys [{arriveTangent=-2.543, interpMode=RCIM_Cubic, leaveTangent=-2.543, tangentMode=RCTM_User, time=0, value=1}, <at 1 -0.05 RCIM_Cubic>]}
      }}
     ScaleCurve=&Multiply_Vector2D_ByFloat(Float=$User.Height, Vector2D=<0.65 1>)
    )
    Scale Sprite Size Mode=::NewEnumerator2
   )
   &ScaleColor(
    Scale Alpha=&FloatFromCurve(
     FloatCurve=DataInterface() {propertyValues {Curve {
        defaultValue=1
        keys [
         {interpMode=RCIM_Cubic, tangentMode=RCTM_User, time=0, value=0}
         {interpMode=RCIM_Cubic, tangentMode=RCTM_User, time=0.5, value=1}
         {tangentMode=RCTM_User, time=1, value=0}
        ]
       }}}
     Scale Curve=0.7
    )
    ScaleA=true
    ScaleRGB=true
   )
  ]
  rendererClasses [&NiagaraSpriteRendererProperties]
  renderers [{rendererClass=&NiagaraSpriteRendererProperties}]
 }
 falling_sprites {
  simTarget=GPUComputeSim
  emitterUpdateScript [
   &EmitterState(
    Enable Distance Culling=true
    Life Cycle Mode=::NewEnumerator1
    Loop Duration=2
    MaxDistance=8000
    MinDistance=2000
    Module.bUseMaxDistance=true
    Module.bUseMinDistance=true
    Scalability Mode=::NewEnumerator1
    UseLoopDelay=false
   )
   &SpawnRate(SpawnRate=&Multiply_Float(A=60, B=$User.Height), Use Spawn Probability=false)
  ]
  particleSpawnScript [
   &InitializeParticle(
    Color=<rgba 2 0.9088 0.3027 1>
    Lifetime Max=4
    Lifetime Min=3
    Lifetime Mode=::NewEnumerator1
    Mass Mode=::NewEnumerator3
    Position Offset=<0 0 -800>
    Sprite Rotation Mode=::NewEnumerator2
    Sprite Size Mode=::NewEnumerator1
    Uniform Sprite Size Max=16
    Uniform Sprite Size Min=8
    UsePositionOffset=true
    Write Material Random=true
   )
   &AddVelocity(
    Cone Angle=1
    Cone Axis=<0 0 -1>
    Module.UseConeFalloff=false
    Module.UseVelDistribution=false
    Override Seed=false
    Velocity Mode=::NewEnumerator2
    Velocity Speed=&RandomRangeFloat(
     Minimum=&Multiply_Float(A=1, B=$User.Height)
     Maximum=&Multiply_Float(A=2, B=$User.Height)
    )
   )
   &ShapeLocation(MaskSpawnGroup=false, Override Seed=false, Sphere Radius=240, UseOwnerScale=true) {module=ShapeLocation001}
   &ShapeLocation(
    Box Midpoint=<0.5 0.5 1>
    Box Size=&MakeVector(X=80, Y=80, Z=&Multiply_Float(A=4000, B=$User.Height))
    MaskSpawnGroup=false
    Override Seed=false
    Shape Primitive=::NewEnumerator2
    Surface Only=false
    UseOwnerScale=true
   )
  ]
  particleUpdateScript [
   &ParticleState()
   &CurlNoiseForce(Noise Frequency=1, Noise Strength=100, Pan Noise=true, Pan Noise Field=<0.1 0.05 -0.2>)
   &Drag(Drag=&RandomRangeFloat(Minimum=0.1, Maximum=0.2), Use Linear Drag=true, Use Rotational Drag=false)
   &ScaleColor(
    Scale Alpha=&FloatFromCurve(
     FloatCurve=DataInterface() {propertyValues {Curve {keys [
         {interpMode=RCIM_Cubic, leaveTangent=4.5453, tangentMode=RCTM_Break, time=0, value=0}
         {interpMode=RCIM_Cubic, tangentMode=RCTM_Break, time=0.5, value=1}
         {arriveTangent=-4.8588, interpMode=RCIM_Cubic, tangentMode=RCTM_Break, time=1, value=0}
        ]}}}
     Scale Curve=&RandomRangeFloat(Minimum=1, Maximum=1.5)
    )
    ScaleA=true
    ScaleRGB=true
   )
   &VortexForce(Vortex Force Amount=50000) {enabled=false}
   /Niagara/Modules/Update/Velocity/VortexVelocity(
    Influence Falloff Radius=49999998976
    Module.Use Influence Falloff=false
    Velocity Amount:NiagaraFloat=&RandomRangeFloat(Minimum=0, Maximum=250)
    Vortex Origin:Vector3f=&Lerp_Vector(A=Linked() {linkedVariable {name=Engine.Owner.Position, type=NiagaraPosition}}, B=<100 100 100>, Alpha=$Particles.Initial.MaterialRandom)
   ) {enabled=false}
   &VortexForce(
    UseOriginOffset=false
    Vortex Origin:NiagaraPosition=&Lerp_Vector(A=Linked() {linkedVariable {name=Engine.Owner.Position, type=NiagaraPosition}}, B=<100 100 100>, Alpha=$Particles.Initial.MaterialRandom)
   ) {enabled=false, module=VortexForce001}
   &PointAttractionForce(
    Attraction Radius=1000
    Attraction Strength=6
    Attractor Position Offset:Vector3f=&MakeVector(X=0, Y=0, Z=&Multiply_Float(A=-3800, B=$User.Height))
    Kill Within Radius=false
    Module.Use Falloff=true
    UseAttractorOffset=true
   ) {enabled=false}
   &SolveForcesAndVelocity(Clamp Velocity=false, Limit Acceleration=false)
   &ScaleSpriteSize(Scale Factor=&Vector2DFromFloat(Value=&FloatFromCurve(FloatCurve=DataInterface() {propertyValues {Curve {keys [
          {interpMode=RCIM_Cubic, tangentMode=RCTM_User, time=0, value=0}
          {interpMode=RCIM_Cubic, tangentMode=RCTM_User, time=0.5, value=1}
          <at 1 0 RCIM_Cubic>
         ]}}})))
  ]
  rendererClasses [&NiagaraSpriteRendererProperties]
  renderers [{rendererClass=&NiagaraSpriteRendererProperties}]
 }
}
summary {
 systemName=NS_Orb_DropShard_Beam
 emitters [
  {emitterName=PickupBeam, rendererClasses [&NiagaraMeshRendererProperties]}
  {emitterName=PlasmaWave, rendererClasses []}
  {emitterName=CoreBleedLight, rendererClasses [&NiagaraMeshRendererProperties, &NiagaraSpriteRendererProperties]}
  {emitterName=Rays, simTarget=GPUComputeSim, rendererClasses [&NiagaraSpriteRendererProperties]}
  {emitterName=Impact_Flare, simTarget=GPUComputeSim, rendererClasses [&NiagaraSpriteRendererProperties]}
  {emitterName=UpwardStreaks, rendererClasses [&NiagaraSpriteRendererProperties]}
  {emitterName=EnergyLeak001, rendererClasses [&NiagaraSpriteRendererProperties]}
  {emitterName=Cylinder, rendererClasses [&NiagaraMeshRendererProperties]}
  {emitterName=Light, rendererClasses [&NiagaraLightRendererProperties]}
  {emitterName=DarkSPline001, rendererClasses [&NiagaraMeshRendererProperties]}
  {emitterName=falling_sprites, simTarget=GPUComputeSim, rendererClasses [&NiagaraSpriteRendererProperties]}
  {emitterName=Light_End, rendererClasses [&NiagaraLightRendererProperties]}
  {emitterName=Cylinder_Low, rendererClasses [&NiagaraMeshRendererProperties]}
 ]
 userVariables [
  User.Height {type=NiagaraFloat, defaultValue=1, description=""}
  User.SmallBeam {type=NiagaraBool}
 ]
}
