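# Niagara system for an orb tractor beam: a warm orange column with core flare, energy leak and impact flare sprites.
# Mesh beam emitters scale by User.Height and User.SmallBeam, with lights on the central cylinder.
# System orb offsets reposition beam, flare and leak.
# Drifting spline meshes and vortex-pulled floating rocks surround it, while camera-distance scalability curves fade, shrink or thin most layers at close range.
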NiagaraSystem /DragonCartNarrative/Assets/VFX/Orb_Drop_Shard/NS_Orb_Tractor_Beam
dependencies {
 usedDataInterfaces [
  &NiagaraDataInterfaceCurve
  &NiagaraDataInterfaceVectorCurve
  &NiagaraDataInterfaceColorCurve
  &NiagaraDataInterfaceCamera
  &NiagaraDataInterfaceVector2DCurve
  &NiagaraDataInterfaceSpriteRendererInfo
 ]
 usedDynamicInputs [
  &RandomVector
  &MakeFloatFromVector2d
  &self:ScratchDynamicInput
  &TransformVector
 ]
 usedModules [
  &self:SystemSpawnScript.SystemScriptSource.SystemScriptGraph.NiagaraNodeAssignment_0.SetVariables_#1
  &self:SystemSpawnScript.SystemScriptSource.SystemScriptGraph.NiagaraNodeAssignment_1.SetVariables_#2
  &SystemState
  &self:SystemSpawnScript.SystemScriptSource.SystemScriptGraph.NiagaraNodeAssignment_2.SetVariables_#3
  &self:SystemSpawnScript.SystemScriptSource.SystemScriptGraph.NiagaraNodeAssignment_3.SetVariables_#4
 ]
}
emitters {
 Beam_Sides {
  emitterUpdateScript [
   &EmitterState(
    Enable Distance Culling=true
    Inactive Response=::NewEnumerator1
    Life Cycle Mode=::NewEnumerator1
    Loop Behavior=::NewEnumerator1
    Loop Delay=1.7
    Loop Duration Mode=::NewEnumerator1
    MaxDistance=100000
    Min Distance Response=::SleepAndLetParticlesFinish
    MinDistance=3000
    Module.bUseMaxDistance=true
    Module.bUseMinDistance=true
    Scalability Mode=::NewEnumerator1
    UseLoopDelay=false
   )
   &SpawnBurst_Instantaneous(Use Spawn Probability=false, UseLoopCountLimit=false)
  ]
  particleSpawnScript [
   &InitializeParticle(
    Color=<rgba 1 0.2326 0.0297 0.5>
    Lifetime=0.4
    Mesh Scale=&MakeVector(X=25, Y=20, Z=&Multiply_Float(A=24, B=$User.Height))
    Mesh Scale Mode=::NewEnumerator4
    UsePositionOffset=false
    Write Material Random=true
   )
   /SetVariables(Particles.PickUpBeam_Size:Vector3f=&VectorFromFloat(Value=&Multiply_Float(A=22, B=$User.Height))) {isSetParameters=true, module=SetVariables_#5}
  ]
  particleUpdateScript [
   &SolveForcesAndVelocity(Clamp Velocity=false, Limit Acceleration=false)
   &ParticleState(Kill Particles When Lifetime Has Elapsed=false)
   &ScaleMeshSize(Scale Factor=&VectorFromCurve(
     VectorCurve=DataInterface() {propertyValues {
       XCurve {keys [{interpMode=RCIM_Cubic, leaveTangent=5.3095, tangentMode=RCTM_Break, time=0, value=0}, <at 0.5 1>]}
       YCurve {keys [{arriveTangent=0.035, interpMode=RCIM_Cubic, leaveTangent=0.035, tangentMode=RCTM_User, time=0, value=0.05}, <at 1 1>]}
       ZCurve {keys [{arriveTangent=2.676, interpMode=RCIM_Cubic, leaveTangent=2.676, tangentMode=RCTM_User, time=0, value=0.05}, <at 1 1>]}
      }}
     Scale Curve=&VectorFromFloat(Value=&MakeCustomFloatFromBool(Bool=$User.SmallBeam, True Float=0.7, False Float=1))
    ))
   &ScaleMeshSize(Scale Factor=&VectorFromFloat(Value=&FloatFromCurve(
      FloatCurve=DataInterface() {propertyValues {Curve {keys [<at 1 1>]}}}
      Scale Curve=&Lerp_Float(
       A=0
       B=1
       Alpha=/Niagara/DynamicInputs/Scalability/ScalabilityDistanceBasedFloat(
        Curve:NiagaraDataInterfaceCurve=DataInterface() {propertyValues {Curve {defaultValue=0, keys [{tangentMode=RCTM_User, time=0, value=0}, <at 0.1 1>, {tangentMode=RCTM_User, time=1, value=1}]}}}
        CurveIndex:NiagaraFloat=$Emitter.Scalability.DistanceFraction
        Bias:NiagaraFloat=0
        Scale:NiagaraFloat=1
       )
      )
     ))) {module=ScaleMeshSize001}
   /SetVariables(Particles.Position=&AddVectorToPosition(Position=<0 0 0>, Vector=$System.OrbWPO)) {isSetParameters=true, module=SetVariables_#6}
   &ScaleColor(
    Scale Alpha=&FloatFromCurve(FloatCurve=DataInterface() {propertyValues {Curve {keys [{interpMode=RCIM_Cubic, leaveTangent=4.9886, tangentMode=RCTM_Break, time=0, value=0.1608}, <at 0.3 1>]}}})
    ScaleA=true
    ScaleRGB=true
   )
   /SetVariables(Particles.PickUPBeam_Dynamic_01:Vector4f=&MakeVector4(
     X=&FloatFromCurve(
      FloatCurve=DataInterface() {propertyValues {Curve {
         defaultValue=100
         keys [{interpMode=RCIM_Cubic, tangentMode=RCTM_User, time=0, value=0}, {interpMode=RCIM_Cubic, tangentMode=RCTM_User, time=1, value=33}]
        }}}
      Scale Curve=80
     )
     Y=1
     Z=1
     W=1
    )) {isSetParameters=true, module=SetVariables_#7}
   /SetVariables(Particles.PickUpbeam_Color:LinearColor=&Multiply_LinearColorByFloat(Linear Color=<rgba 1 0.4048 0.1927 0.35>, Float=1)) {isSetParameters=true, module=SetVariables_#8}
   /SetVariables(Particles.PickUpBeam_Orientation001:Quat4f=<1 1 0 1>) {isSetParameters=true}
   /SetVariables(Particles.PickUpBeam_Orientation002:Quat4f=<-1 1 0 1>) {isSetParameters=true}
   /SetVariables(Particles.PickUpBeam_Orientation003:Quat4f=<0.8 0.45 -0.35 1>) {isSetParameters=true}
   /SetVariables(Particles.PickUpBeam_Orientation004:Quat4f=<-0.2 0 0.5 0>) {isSetParameters=true}
   &FN_Light_Attributes(
    Color=$Particles.Light_Color
    Create Color Attribute=true
    Create Radius Attribute=true
    Radius=2000
    Scale Color=&FloatFromCurve(
     FloatCurve=DataInterface() {propertyValues {Curve {
        defaultValue=1
        keys [{interpMode=RCIM_Cubic, tangentMode=RCTM_User, time=0, value=0}, {interpMode=RCIM_Cubic, tangentMode=RCTM_User, time=0.5, value=1}]
       }}}
     Scale Curve=16
    )
   )
   &DynamicMaterialParameters(
    Index 0 Param 1=1
    Index 0 Param 4=0.2
    Module.Index0Float0Write=true
    Module.Index0Float1Write=true
    Module.Index0Float2Write=true
    Module.Index0Float3Write=true
   )
  ]
  rendererClasses [&NiagaraMeshRendererProperties]
  renderers [
   {rendererClass=&NiagaraMeshRendererProperties, rendererIndex=0}
   {rendererClass=&NiagaraMeshRendererProperties, rendererIndex=1}
   {rendererClass=&NiagaraMeshRendererProperties, rendererIndex=2}
   {rendererClass=&NiagaraMeshRendererProperties, rendererIndex=3}
  ]
 }
 CoreBleed_Flare {
  emitterUpdateScript [
   &EmitterState(
    Enable Distance Culling=true
    Inactive Response=::NewEnumerator1
    Life Cycle Mode=::NewEnumerator1
    Loop Behavior=::NewEnumerator1
    Loop Delay=1.7
    Loop Duration Mode=::NewEnumerator1
    MaxDistance=100000
    Min Distance Response=::SleepAndLetParticlesFinish
    MinDistance=2000
    Module.bUseMaxDistance=true
    Module.bUseMinDistance=true
    Scalability Mode=::NewEnumerator1
    UseLoopDelay=false
   )
   &SpawnBurst_Instantaneous(Use Spawn Probability=false, UseLoopCountLimit=false)
  ]
  particleSpawnScript [
   &InitializeParticle(
    Color=<rgba 10 2.6079 0.7812 1>
    Lifetime=0.4
    Mesh Scale=<9 9 3>
    Mesh Scale Mode=::NewEnumerator4
    Sprite Size Mode=::NewEnumerator3
    Uniform Sprite Size=1100
    UsePositionOffset=false
    Write Material Random=true
   )
   &InitialMeshOrientation(Mesh Orientation Mode=::NewEnumerator2, Orientation Vector=<0 0 1>, Use Rotation Vector=false)
   &SpriteFacingAndAlignment(Sprite Facing=<0 0 1>, WriteAlignment=false, WriteFacing=true)
   &DynamicMaterialParameters(
    Index 3 Param 1=2.5
    Index 3 Param 2=0.25
    Module.Index3Float0Write=true
    Module.Index3Float1Write=true
    Module.Index3Float2Write=false
    Module.Index3Float3Write=false
    Write Parameter Index 0=::NewEnumerator0
    Write Parameter Index 3=::NewEnumerator1
   )
  ]
  particleUpdateScript [
   &ParticleState(Kill Particles When Lifetime Has Elapsed=false)
   &ScaleSpriteSize(
    Uniform Curve Scale=2
    Uniform Curve Sprite Scale=DataInterface() {propertyValues {Curve {keys [{arriveTangent=2.288, interpMode=RCIM_Cubic, time=0, value=0}, {tangentMode=RCTM_User, time=1, value=1}]}}}
   )
   &ScaleSpriteSize(
    Scale Sprite Size Mode=::NewEnumerator0
    Uniform Scale Factor=&MakeFloatFromVector(
     Vector=&VectorFromFloat(Value=&Lerp_Float(
       A=0
       B=1
       Alpha=&ScalabilityDistanceBasedFloat(
        Curve:NiagaraDataInterfaceCurve=DataInterface() {propertyValues {Curve {defaultValue=0, keys [{tangentMode=RCTM_User, time=0, value=0}, <at 0.05 1>, {tangentMode=RCTM_User, time=1, value=1}]}}}
        CurveIndex:NiagaraFloat=$Emitter.Scalability.DistanceFraction
        Bias:NiagaraFloat=0
        Scale:NiagaraFloat=1
       )
      ))
     Channel=::NewEnumerator0
    )
   ) {module=ScaleSpriteSize001}
   &ScaleMeshSize(Scale Factor=&VectorFromCurve(
     VectorCurve=DataInterface() {propertyValues {
       XCurve {keys [{arriveTangent=2.3661, interpMode=RCIM_Cubic, time=0, value=0}, <at 1 1>]}
       YCurve {keys [{arriveTangent=2.3661, interpMode=RCIM_Cubic, time=0, value=0}, <at 1 1>]}
       ZCurve {keys [{arriveTangent=2.3661, interpMode=RCIM_Cubic, time=0, value=0}, <at 1 1>]}
      }}
     Scale Curve=&VectorFromFloat(Value=&MakeCustomFloatFromBool(Bool=$User.SmallBeam, True Float=-1, False Float=-2))
    ))
   &ScaleColor(
    Scale Alpha=&FloatFromCurve(
     FloatCurve=DataInterface() {propertyValues {Curve {keys [{interpMode=RCIM_Cubic, leaveTangent=4.9886, tangentMode=RCTM_Break, time=0, value=0.1608}, <at 0.3 1>]}}}
     Scale Curve=&Waveform("Waveform [1]"=::NewEnumerator12, "[1] Frequency"=9, "[1] Amplitude Min/Max"=<0.75 1>)
    )
    ScaleA=true
    ScaleRGB=true
   )
   /SetVariables(Particles.Position=&AddVectorToPosition(Position=<0 0 -76>, Vector=$System.OrbWPOLocal)) {isSetParameters=true}
   &CameraOffset(Camera Offset Amount=10)
  ]
  rendererClasses [&NiagaraMeshRendererProperties]
  renderers [{rendererClass=&NiagaraMeshRendererProperties, rendererIndex=0}]
 }
 Cylinder004 {
  emitterUpdateScript [
   &EmitterState(
    Enable Distance Culling=true
    Inactive Response=::NewEnumerator1
    Life Cycle Mode=::NewEnumerator1
    Loop Behavior=::NewEnumerator1
    Loop Delay=1.7
    Loop Duration Mode=::NewEnumerator1
    MaxDistance=100000
    Min Distance Response=::SleepAndLetParticlesFinish
    MinDistance=2000
    Module.bUseMaxDistance=true
    Module.bUseMinDistance=true
    Scalability Mode=::NewEnumerator1
    UseLoopDelay=false
   )
   &SpawnBurst_Instantaneous(Use Spawn Probability=false, UseLoopCountLimit=false)
  ]
  particleSpawnScript [
   &InitializeParticle(
    Color=<rgba 1 0.2326 0.0297 0.5>
    Lifetime=0.4
    Mesh Scale=&MakeVector(X=20, Y=20, Z=&Multiply_Float(A=23, B=$User.Height))
    Mesh Scale Mode=::NewEnumerator4
    UsePositionOffset=false
    Write Material Random=true
   )
   /SetVariables(Particles.Light_Color:LinearColor=<rgba 1.5 0.226 0 1>) {isSetParameters=true}
  ]
  particleUpdateScript [
   &SolveForcesAndVelocity(Clamp Velocity=false, Limit Acceleration=false)
   &ParticleState(Kill Particles When Lifetime Has Elapsed=false)
   &ScaleMeshSize(Scale Factor=&VectorFromCurve(
     VectorCurve=DataInterface() {propertyValues {
       XCurve {keys [{interpMode=RCIM_Cubic, leaveTangent=5.3095, tangentMode=RCTM_Break, time=0, value=0}, <at 0.5 1>]}
       YCurve {keys [{arriveTangent=0.035, interpMode=RCIM_Cubic, leaveTangent=0.035, tangentMode=RCTM_User, time=0, value=0.05}, <at 1 1>]}
       ZCurve {keys [{arriveTangent=2.676, interpMode=RCIM_Cubic, leaveTangent=2.676, tangentMode=RCTM_User, time=0, value=0.05}, <at 1 1>]}
      }}
     Scale Curve=&VectorFromFloat(Value=&MakeCustomFloatFromBool(Bool=$User.SmallBeam, True Float=0.7, False Float=1))
    ))
   /SetVariables(Particles.Position=&AddVectorToPosition(Position=<0 0 0>, Vector=$System.OrbWPO)) {isSetParameters=true, module=SetVariables_#6}
   &ScaleColor(
    Scale Alpha=&FloatFromCurve(FloatCurve=DataInterface() {propertyValues {Curve {keys [{interpMode=RCIM_Cubic, leaveTangent=4.9886, tangentMode=RCTM_Break, time=0, value=0.1608}, <at 0.3 1>]}}})
    ScaleA=true
    ScaleRGB=true
   )
   /SetVariables(Particles.PickUpBeam_Size:Vector3f=&VectorFromFloat(Value=&Lerp_Float(
      A=0
      B=22
      Alpha=&ScalabilityDistanceBasedFloat(
       Curve:NiagaraDataInterfaceCurve=DataInterface() {propertyValues {Curve {defaultValue=0, keys [{tangentMode=RCTM_User, time=0, value=0}, <at 0.05 1>, {tangentMode=RCTM_User, time=1, value=1}]}}}
       CurveIndex:NiagaraFloat=$Emitter.Scalability.DistanceFraction
       Bias:NiagaraFloat=0
       Scale:NiagaraFloat=1
      )
     ))) {isSetParameters=true, module=SetVariables_#5}
   /SetVariables(Particles.PickUPBeam_Dynamic_01:Vector4f=&MakeVector4(
     X=&FloatFromCurve(
      FloatCurve=DataInterface() {propertyValues {Curve {
         defaultValue=100
         keys [{interpMode=RCIM_Cubic, tangentMode=RCTM_User, time=0, value=0}, {interpMode=RCIM_Cubic, tangentMode=RCTM_User, time=1, value=33}]
        }}}
      Scale Curve=200
     )
     Y=1
     Z=1
     W=1
    )) {isSetParameters=true, module=SetVariables_#7}
   /SetVariables(Particles.PickUpbeam_Color:LinearColor=&Multiply_LinearColorByFloat(
     Linear Color=<rgba 1 0.4048 0.1927 0.6609>
     Float=&Lerp_Float(
      A=0
      B=1
      Alpha=&ScalabilityDistanceBasedFloat(
       Curve:NiagaraDataInterfaceCurve=DataInterface() {propertyValues {Curve {defaultValue=0, keys [{tangentMode=RCTM_User, time=0, value=0}, <at 0.02 1>, {tangentMode=RCTM_User, time=1, value=1}]}}}
       CurveIndex:NiagaraFloat=$Emitter.Scalability.DistanceFraction
       Bias:NiagaraFloat=0
       Scale:NiagaraFloat=1
      )
     )
    )) {isSetParameters=true, module=SetVariables_#8}
   /SetVariables(Particles.PickUpBeam_Orientation:Quat4f=<0 1 0 1>) {isSetParameters=true}
   &FN_Light_Attributes(
    Color=$Particles.Light_Color
    Create Color Attribute=true
    Create Radius Attribute=true
    Radius=2250
    Scale Color=&FloatFromCurve(
     FloatCurve=DataInterface() {propertyValues {Curve {
        defaultValue=1
        keys [{interpMode=RCIM_Cubic, tangentMode=RCTM_User, time=0, value=0}, {interpMode=RCIM_Cubic, tangentMode=RCTM_User, time=0.5, value=1}]
       }}}
     Scale Curve=16
    )
   )
   /SetVariables(Particles.LightUP_Pos:NiagaraPosition=&AddVectorToPosition(Position=<0 0 -300>, Vector=$System.OrbWPOLocal)) {isSetParameters=true}
   /SetVariables(Particles.LightDown_Pos:NiagaraPosition=&AddVectorToPosition(
     Position=&MakeVector(X=0, Y=0, Z=&Multiply_Float(A=-3000, B=$User.Height))
     Vector=$System.OrbWPOLocal
    )) {isSetParameters=true}
   /SetVariables(Particles.LightRadius=2600) {isSetParameters=true}
   &DynamicMaterialParameters(
    Index 0 Param 1=1
    Index 0 Param 4=0.2
    Module.Index0Float0Write=true
    Module.Index0Float1Write=true
    Module.Index0Float2Write=true
    Module.Index0Float3Write=true
   )
  ]
  rendererClasses [&NiagaraMeshRendererProperties, &NiagaraLightRendererProperties]
  renderers [
   {rendererClass=&NiagaraMeshRendererProperties, rendererIndex=0}
   {rendererClass=&NiagaraMeshRendererProperties, rendererIndex=1}
   {rendererClass=&NiagaraLightRendererProperties, rendererIndex=2}
   {rendererClass=&NiagaraLightRendererProperties, rendererIndex=3}
  ]
 }
 DarkSPline {
  emitterUpdateScript [
   &EmitterState(
    Enable Distance Culling=true
    Life Cycle Mode=::NewEnumerator1
    Loop Delay=0.8
    MaxDistance=100000
    Min Distance Response=::SleepAndLetParticlesFinish
    MinDistance=3500
    Module.bApplySpawnCountScaleByDistance=true
    Module.bUseMaxDistance=true
    Module.bUseMinDistance=true
    Scalability Mode=::NewEnumerator1
    Scale Spawn Count=true
    Spawn Count Scale By Distance=DataInterface() {propertyValues {Curve {keys [<at 0 0>,<at 0.2 1>,<at 1 1>]}}}
    UseLoopDelay=false
   )
   &SpawnBurst_Instantaneous(Spawn Count=2, Use Spawn Probability=false, UseLoopCountLimit=false)
   &SpawnRate(SpawnRate=&Multiply_Float(A=3, B=$User.Height), Use Spawn Probability=false)
  ]
  particleSpawnScript [
   &InitializeParticle(
    Color=<rgba 5 1.4718 0.3906 0.7>
    Lifetime Max=6
    Lifetime Min=3
    Lifetime Mode=::NewEnumerator1
    Mass Max=0.2
    Mass Min=0
    Mass Mode=::NewEnumerator3
    Mesh Scale Max=<15 15 10>
    Mesh Scale Min=<-15 -15 9>
    Mesh Scale Mode=::NewEnumerator2
    Sprite Rotation Mode=::NewEnumerator2
    UsePositionOffset=false
    Write Material Random=true
   )
   &InitialMeshOrientation(
    Mesh Orientation Mode=::NewEnumerator2
    Rotation=&RandomRangeVector(Minimum=<0 0 0>, Maximum=<0 0 1>)
    Use Rotation Vector=true
   )
   &AddVelocity(
    Cone Angle=1
    Cone Axis=<0 0 1>
    Module.UseConeFalloff=false
    Module.UseVelDistribution=false
    Override Seed=false
    Velocity Mode=::NewEnumerator2
    Velocity Speed=&RandomRangeFloat(Minimum=50, Maximum=100)
   )
   &ShapeLocation(
    Cylinder Height=&Multiply_Float(A=100, B=$User.Height)
    Cylinder Height Midpoint=1
    Cylinder Radius=200
    MaskSpawnGroup=false
    Override Seed=false
    Shape Primitive=::NewEnumerator1
    Surface Only=false
    UseOwnerScale=true
   ) {module=ShapeLocation001}
  ]
  particleUpdateScript [
   &ParticleState()
   &ScaleMeshSize(Scale Factor=&VectorFromCurve(
     VectorCurve=DataInterface() {propertyValues {
       XCurve {keys [
         {arriveTangent=-0.0178, interpMode=RCIM_Cubic, leaveTangent=-0.0178, tangentMode=RCTM_User, time=0, value=1.6}
         {arriveTangent=-2.5718, interpMode=RCIM_Cubic, leaveTangent=-2.5718, tangentMode=RCTM_User, time=0.3, value=1.3}
         {arriveTangent=0.0224, interpMode=RCIM_Cubic, leaveTangent=0.0224, tangentMode=RCTM_User, time=1, value=0.1}
        ]}
       YCurve {keys [
         {arriveTangent=-0.2493, interpMode=RCIM_Cubic, leaveTangent=-0.2493, tangentMode=RCTM_User, time=0, value=1.6}
         {arriveTangent=-2.1592, interpMode=RCIM_Cubic, leaveTangent=-2.1592, tangentMode=RCTM_User, time=0.3, value=1.3}
         {arriveTangent=0.0224, interpMode=RCIM_Cubic, leaveTangent=0.0224, tangentMode=RCTM_User, time=1, value=0.1}
        ]}
       ZCurve {
        defaultValue=3
        keys [{interpMode=RCIM_Cubic, tangentMode=RCTM_User, time=0, value=3}, {interpMode=RCIM_Cubic, tangentMode=RCTM_User, time=1, value=0}]
       }
      }}
     Scale Curve=&Multiply_VectorByFloat(Vector=<45 45 30>, Float=0.07)
    ))
   &Drag(Drag=&RandomRangeFloat(Minimum=0.25, Maximum=0.35), Use Linear Drag=true, Use Rotational Drag=false)
   &ScaleColor(
    Scale Alpha=&FloatFromCurve(
     FloatCurve=DataInterface() {propertyValues {Curve {keys [
         {arriveTangent=0.0662, interpMode=RCIM_Cubic, leaveTangent=0.0662, tangentMode=RCTM_User, time=0, value=0}
         {interpMode=RCIM_Cubic, tangentMode=RCTM_User, time=0.8, value=1.5}
         {tangentMode=RCTM_User, time=1, value=0}
        ]}}}
     Scale Curve=1.8
    )
    ScaleA=true
    ScaleRGB=true
   )
   &CameraOffset(Camera Offset Amount=&Multiply_Float(
     A=/Game/Effects/Niagara/Modules/Math/FN_CameraDistance(
      0Dist:NiagaraFloat=0
      1Dist:NiagaraFloat=100
      Camera Query:NiagaraDataInterfaceCamera=DataInterface() {propertyValues {bRequireCurrentFrameData=false}}
      Position:NiagaraPosition=$Particles.Position
     )
     B=300
    ))
   &GravityForce(Gravity=&RandomRangeVector(Minimum=<0 0 10>, Maximum=<0 0 30>))
   &SolveForcesAndVelocity(Clamp Velocity=false, Limit Acceleration=false)
   &DynamicMaterialParameters(
    Index 0 Param 1=-0.1
    Index 0 Param 3=1
    Index 1 Param 1=1
    Index 1 Param 2=1
    Index 1 Param 3=1
    Module.Index0Float0Write=true
    Module.Index0Float1Write=true
    Module.Index0Float2Write=true
    Module.Index0Float3Write=true
    Module.Index1Float0Write=false
    Module.Index1Float1Write=false
    Module.Index1Float2Write=false
    Module.Index1Float3Write=false
    Write Parameter Index 1=::NewEnumerator1
   )
   &ScaleColor(
    Scale Alpha=&Lerp_Float(
     A=0
     B=1
     Alpha=&ScalabilityDistanceBasedFloat(
      Curve:NiagaraDataInterfaceCurve=DataInterface() {propertyValues {Curve {defaultValue=0, keys [{tangentMode=RCTM_User, time=0, value=0}, <at 0.1 1>, {tangentMode=RCTM_User, time=1, value=1}]}}}
      CurveIndex:NiagaraFloat=$Emitter.Scalability.DistanceFraction
      Bias:NiagaraFloat=0
      Scale:NiagaraFloat=1
     )
    )
    ScaleA=true
    ScaleRGB=false
   ) {module=ScaleColor001}
  ]
  rendererClasses [&NiagaraMeshRendererProperties]
  renderers [{rendererClass=&NiagaraMeshRendererProperties, rendererIndex=0}]
 }
 DarkSPline001 {
  emitterUpdateScript [
   &EmitterState(
    Enable Distance Culling=true
    Life Cycle Mode=::NewEnumerator1
    Loop Delay=0.8
    MaxDistance=100000
    Min Distance Response=::SleepAndLetParticlesFinish
    MinDistance=3500
    Module.bApplySpawnCountScaleByDistance=true
    Module.bUseMaxDistance=true
    Module.bUseMinDistance=true
    Scalability Mode=::NewEnumerator1
    Scale Spawn Count=true
    Spawn Count Scale By Distance=DataInterface() {propertyValues {Curve {keys [<at 0 0>,<at 0.2 1>,<at 1 1>]}}}
    UseLoopDelay=false
   )
   &SpawnBurst_Instantaneous(Spawn Count=2, Use Spawn Probability=false, UseLoopCountLimit=false)
   &SpawnRate(SpawnRate=&Multiply_Float(A=1, B=$User.Height), Use Spawn Probability=false)
  ]
  particleSpawnScript [
   &InitializeParticle(
    Color=<rgba 10 2.3455 0 0.6>
    Lifetime Max=6
    Lifetime Min=5
    Lifetime Mode=::NewEnumerator1
    Mass Max=0.2
    Mass Min=0
    Mass Mode=::NewEnumerator3
    Mesh Scale Max=<20 20 10>
    Mesh Scale Min=<-20 -20 10>
    Mesh Scale Mode=::NewEnumerator2
    Position Offset=<0 0 -2200>
    Sprite Rotation Mode=::NewEnumerator2
    UsePositionOffset=true
    Write Material Random=true
   )
   &InitialMeshOrientation(
    Mesh Orientation Mode=::NewEnumerator2
    Rotation=&RandomRangeVector(Minimum=<0 0 0>, Maximum=<0 0 1>)
    Use Rotation Vector=true
   )
   &AddVelocity(
    Cone Angle=1
    Cone Axis=<0 0 -1>
    Module.UseConeFalloff=false
    Module.UseVelDistribution=false
    Override Seed=false
    Velocity Mode=::NewEnumerator2
    Velocity Speed=&RandomRangeFloat(Minimum=1, Maximum=2)
   ) {enabled=false}
   &ShapeLocation(
    Cylinder Height=&Multiply_Float(A=200, B=$User.Height)
    Cylinder Height Midpoint=1
    Cylinder Radius=1000
    MaskSpawnGroup=false
    Override Seed=false
    Shape Primitive=::NewEnumerator1
    Surface Only=false
    UseOwnerScale=true
   ) {module=ShapeLocation001}
  ]
  particleUpdateScript [
   &ParticleState()
   &ScaleMeshSize(Scale Factor=&VectorFromCurve(
     VectorCurve=DataInterface() {propertyValues {
       XCurve {keys [
         {arriveTangent=-0.0178, interpMode=RCIM_Cubic, leaveTangent=-0.0178, tangentMode=RCTM_User, time=0, value=2}
         {arriveTangent=-2.5718, interpMode=RCIM_Cubic, leaveTangent=-2.5718, tangentMode=RCTM_User, time=0.5, value=1.5}
         {arriveTangent=0.0224, interpMode=RCIM_Cubic, leaveTangent=0.0224, tangentMode=RCTM_User, time=1, value=1}
        ]}
       YCurve {keys [
         {arriveTangent=-0.2493, interpMode=RCIM_Cubic, leaveTangent=-0.2493, tangentMode=RCTM_User, time=0, value=2}
         {arriveTangent=-2.1592, interpMode=RCIM_Cubic, leaveTangent=-2.1592, tangentMode=RCTM_User, time=0.5, value=1.5}
         {arriveTangent=0.0224, interpMode=RCIM_Cubic, leaveTangent=0.0224, tangentMode=RCTM_User, time=1, value=1}
        ]}
       ZCurve {keys [
         {arriveTangent=0.0002, interpMode=RCIM_Cubic, leaveTangent=0.0002, tangentMode=RCTM_User, time=0, value=2}
         {arriveTangent=-1.6771, interpMode=RCIM_Cubic, leaveTangent=-1.6771, tangentMode=RCTM_User, time=0.5, value=1.5}
         {arriveTangent=0.0224, interpMode=RCIM_Cubic, leaveTangent=0.0224, tangentMode=RCTM_User, time=1, value=1}
        ]}
      }}
     Scale Curve=&Multiply_VectorByFloat(Vector=<300 300 50>, Float=0.05)
    ))
   &Drag(Drag=&RandomRangeFloat(Minimum=0.25, Maximum=0.35), Use Linear Drag=true, Use Rotational Drag=false)
   &ScaleColor(
    Scale Alpha=&FloatFromCurve(
     FloatCurve=DataInterface() {propertyValues {Curve {keys [
         {arriveTangent=0.091, interpMode=RCIM_Cubic, leaveTangent=0.091, tangentMode=RCTM_User, time=0, value=0}
         {interpMode=RCIM_Cubic, tangentMode=RCTM_User, time=0.7, value=1}
         {tangentMode=RCTM_User, time=1, value=0}
        ]}}}
     Scale Curve=0.4
    )
    ScaleA=true
    ScaleRGB=true
   )
   &CameraOffset(Camera Offset Amount=&Multiply_Float(
     A=&FN_CameraDistance(
      0Dist:NiagaraFloat=0
      1Dist:NiagaraFloat=100
      Camera Query:NiagaraDataInterfaceCamera=DataInterface() {propertyValues {bRequireCurrentFrameData=false}}
      Position:NiagaraPosition=$Particles.Position
     )
     B=300
    ))
   &GravityForce(Gravity=&RandomRangeVector(Minimum=<0 0 30>, Maximum=<0 0 60>))
   &SolveForcesAndVelocity(Clamp Velocity=false, Limit Acceleration=false)
   &DynamicMaterialParameters(
    Index 0 Param 1=-0.15
    Index 0 Param 3=1
    Module.Index0Float0Write=true
    Module.Index0Float1Write=true
    Module.Index0Float2Write=true
    Module.Index0Float3Write=true
    Module.Index1Float0Write=true
    Module.Index1Float1Write=true
    Module.Index1Float2Write=false
    Module.Index1Float3Write=false
    Write Parameter Index 1=::NewEnumerator1
   )
  ]
  rendererClasses [&NiagaraMeshRendererProperties]
  renderers [{rendererClass=&NiagaraMeshRendererProperties, rendererIndex=0}]
 }
 EnergyLeak {
  simTarget=GPUComputeSim
  emitterUpdateScript [
   &EmitterState(
    Enable Distance Culling=true
    Life Cycle Mode=::NewEnumerator1
    Loop Duration=1
    MaxDistance=100000
    Min Distance Response=::SleepAndLetParticlesFinish
    MinDistance=2000
    Module.bUseMaxDistance=true
    Module.bUseMinDistance=true
    Scalability Mode=::NewEnumerator1
    UseLoopDelay=false
   )
   &SpawnBurst_Instantaneous(Use Spawn Probability=false, UseLoopCountLimit=false)
   &SpawnBurst_Instantaneous(Spawn Group=1, Use Spawn Probability=false, UseLoopCountLimit=false) {module=SpawnBurst_Instantaneous001}
   &SpawnRate(SpawnGroup=2, Use Spawn Probability=false) {enabled=false}
   &SpawnBurst_Instantaneous(
    Spawn Count=&RandomRangeInt(Minimum=1, Maximum=1)
    Spawn Group=3
    Use Spawn Probability=false
    UseLoopCountLimit=false
   ) {enabled=false, module=SpawnBurst_Instantaneous002}
  ]
  particleSpawnScript [
   /SetVariables(Particles.VisibilityTag=&SpawnGroup()) {isSetParameters=true, module=SetVariables_#9}
   &InitializeParticle(
    Color=<rgba 1.2 0.1667 0 1>
    Lifetime=&MaskFloatBySpawnGroup(
     Masked Float=2
     Passthrough Float=&MaskFloatBySpawnGroup(
      Masked Float=5
      Passthrough Float=&MaskFloatBySpawnGroup(Masked Float=5, Passthrough Float=&RandomRangeFloat(Minimum=1, Maximum=2), Spawn Group Mask=2)
      Spawn Group Mask=1
     )
    )
    Mesh Scale=<12.5 12.5 12.5>
    Mesh Scale Mode=::NewEnumerator4
    Sprite Rotation Angle=&MaskFloatBySpawnGroup(Masked Float=0, Passthrough Float=&RandomRangeFloat(Minimum=-360, Maximum=360))
    Sprite Rotation Mode=::NewEnumerator0
    Sprite Size Mode=::NewEnumerator3
    Uniform Sprite Size=5000
    UsePositionOffset=false
    Write Material Random=true
   )
   &DynamicMaterialParameters(
    Index 0 Param 1=1
    Index 0 Param 2=1
    Index 0 Param 3=1
    Index 0 Param 4=1
    Module.Index0Float0Write=true
    Module.Index0Float1Write=true
    Module.Index0Float2Write=true
    Module.Index0Float3Write=true
   )
   &InitialMeshOrientation(
    Mesh Axis To Orient=&RandomRangeVector(Minimum=<0 0 1>, Maximum=<0 0 -1>)
    Mesh Orientation Mode=::NewEnumerator2
    Use Rotation Vector=false
   )
  ]
  particleUpdateScript [
   &ParticleState()
   &ScaleSpriteSize(Uniform Curve Sprite Scale=DataInterface() {propertyValues {Curve {keys [
        {interpMode=RCIM_Cubic, tangentMode=RCTM_Break, time=0, value=0.5}
        {arriveTangent=-0.3016, interpMode=RCIM_Cubic, tangentMode=RCTM_Break, time=0.75, value=1}
       ]}}})
   &ScaleSpriteSize(
    Scale Sprite Size Mode=::NewEnumerator0
    Uniform Scale Factor=&MakeFloatFromVector(
     Vector=&VectorFromFloat(Value=&Lerp_Float(
       A=0
       B=1
       Alpha=&ScalabilityDistanceBasedFloat(
        Curve:NiagaraDataInterfaceCurve=DataInterface() {propertyValues {Curve {defaultValue=0, keys [{tangentMode=RCTM_User, time=0, value=0}, <at 0.1 1>, {tangentMode=RCTM_User, time=1, value=1}]}}}
        CurveIndex:NiagaraFloat=$Emitter.Scalability.DistanceFraction
        Bias:NiagaraFloat=0
        Scale:NiagaraFloat=1
       )
      ))
     Channel=::NewEnumerator0
    )
   ) {module=ScaleSpriteSize001}
   &ScaleMeshSize(Scale Factor=&VectorFromFloat(Value=&Lerp_Float(
      A=0
      B=1
      Alpha=&ScalabilityDistanceBasedFloat(
       Curve:NiagaraDataInterfaceCurve=DataInterface() {propertyValues {Curve {defaultValue=0, keys [{tangentMode=RCTM_User, time=0, value=0}, <at 0.05 1>, {tangentMode=RCTM_User, time=1, value=1}]}}}
       CurveIndex:NiagaraFloat=$Emitter.Scalability.DistanceFraction
       Bias:NiagaraFloat=0
       Scale:NiagaraFloat=1
      )
     )))
   &CurlNoiseForce(
    Noise Frequency=1
    Noise Strength=&MaskFloatBySpawnGroup(
     Masked Float=0
     Passthrough Float=&MaskFloatBySpawnGroup(
      Masked Float=0
      Passthrough Float=&MaskFloatBySpawnGroup(Masked Float=0, Passthrough Float=100, Spawn Group Mask=2)
      Spawn Group Mask=1
     )
    )
    Pan Noise=true
    Pan Noise Field=<0.1 0.05 -0.2>
   )
   &Drag(
    Drag=&MaskFloatBySpawnGroup(
     Masked Float=0
     Passthrough Float=&MaskFloatBySpawnGroup(
      Masked Float=0
      Passthrough Float=&MaskFloatBySpawnGroup(Masked Float=0, Passthrough Float=&RandomRangeFloat(Minimum=0.1, Maximum=0.2), Spawn Group Mask=2)
      Spawn Group Mask=1
     )
    )
    Use Linear Drag=true
    Use Rotational Drag=true
   )
   &Timeline(
    Play Rate=&FloatFromCurve(FloatCurve=DataInterface() {propertyValues {Curve {keys [<at 0 2>,<at 1 1.3>]}}})
    Timeline End=1000
   )
   /SetVariables(Particles.EnergyLeakColor:LinearColor=&Multiply_LinearColorByFloat(
     Linear Color=<rgba 1.2 0.1667 0 1>
     Float=&FloatFromCurve(
      FloatCurve=DataInterface() {propertyValues {Curve {keys [
          {arriveTangent=6.7513, interpMode=RCIM_Cubic, leaveTangent=6.7513, tangentMode=RCTM_User, time=0, value=0}
          {arriveTangent=0.1216, interpMode=RCIM_Cubic, leaveTangent=0.1216, tangentMode=RCTM_User, time=0.2, value=0.9847}
          {arriveTangent=-0.0116, interpMode=RCIM_Cubic, leaveTangent=-0.0116, tangentMode=RCTM_User, time=1, value=0}
         ]}}}
      Scale Curve=1.2
     )
    )) {isSetParameters=true, module=SetVariables_#10}
   /SetVariables(Particles.EnergyLeakSize:Vector2f=&Vector2DFromFloat(Value=&Multiply_Float(
      A=900
      B=&FloatFromCurve(
       FloatCurve=DataInterface() {propertyValues {Curve {keys [
           {interpMode=RCIM_Cubic, tangentMode=RCTM_Break, time=0, value=0.5}
           {arriveTangent=-0.3016, interpMode=RCIM_Cubic, tangentMode=RCTM_Break, time=0.75, value=1}
          ]}}}
       Scale Curve=4
      )
     ))) {enabled=false, isSetParameters=true, module=SetVariables_#11}
   /SetVariables(Particles.EnergyLeak_Dyanmic_01:Vector4f=&MakeVector4(X=&Multiply_Float(A=$Particles.Timeline.Time, B=0.5), Y=0, Z=0, W=0)) {isSetParameters=true, module=SetVariables_#12}
   &CameraOffset(Camera Offset Amount=&Multiply_Float(
     A=&FN_CameraDistance(
      0Dist:NiagaraFloat=0
      1Dist:NiagaraFloat=800
      Camera Query:NiagaraDataInterfaceCamera=DataInterface() {propertyValues {bRequireCurrentFrameData=false}}
      Position:NiagaraPosition=$Particles.Position
     )
     B=1000
    ))
   &ScaleColor(
    Scale Alpha=&FloatFromCurve(FloatCurve=DataInterface() {propertyValues {Curve {keys [
         {arriveTangent=-0.1319, interpMode=RCIM_Cubic, leaveTangent=-0.1319, tangentMode=RCTM_User, time=0, value=0}
         <at 0.5 1 RCIM_Cubic>
         <at 1 0 RCIM_Cubic>
        ]}}})
    Scale RGB=&MakeVectorFromLinearColorRGB(Color=&ColorFromCurve(ColorCurve=DataInterface() {propertyValues {
        AlphaCurve {keys []}
        BlueCurve {keys [<at 0 1.1082 RCIM_Cubic>,<at 0.3877 0.0275>,<at 1 0.025 RCIM_Cubic>]}
        GreenCurve {keys [<at 0 1.7837 RCIM_Cubic>,<at 0.3877 0.6197>,<at 1 0 RCIM_Cubic>]}
        RedCurve {keys [<at 0 2 RCIM_Cubic>,<at 0.3877 1>,<at 1 0.1878 RCIM_Cubic>]}
       }}))
    ScaleA=true
    ScaleRGB=true
   )
   /SetVariables(Particles.Position=&AddVectorToPosition(Position=<0 0 0>, Vector=$System.OrbWPOLocal)) {isSetParameters=true, module=SetVariables_#13}
   &FN_Sprite_Attributes(
    Color=&Multiply_LinearColorByFloat(Linear Color=$Particles.Color, Float=2)
    Create Color Attribute=true
    Create SpriteSize Attribute=true
    SpriteSize=&Multiply_Vector2D(A=$Particles.SpriteSize, B=<1 0.4232>)
    Write to Module Namespace=true
   )
   &SolveForcesAndVelocity(Clamp Velocity=false, Limit Acceleration=false)
  ]
  rendererClasses [&NiagaraSpriteRendererProperties]
  renderers [{rendererClass=&NiagaraSpriteRendererProperties}]
 }
 Floating_Rocks {
  emitterUpdateScript [
   &EmitterState(
    Enable Distance Culling=true
    Life Cycle Mode=::NewEnumerator1
    Loop Duration=1
    MaxDistance=100000
    Min Distance Response=::SleepAndLetParticlesFinish
    MinDistance=3000
    Module.bApplySpawnCountScaleByDistance=true
    Module.bUseMaxDistance=true
    Module.bUseMinDistance=true
    Scalability Mode=::NewEnumerator1
    Scale Spawn Count=true
    Spawn Count Scale By Distance=DataInterface() {propertyValues {Curve {keys [<at 0 0.5>,<at 0.2 1>,<at 1 1>]}}}
    UseLoopDelay=false
   )
   &SpawnRate(SpawnRate=&RandomRangeFloat(Minimum=2, Maximum=3), Use Spawn Probability=false)
  ]
  particleSpawnScript [
   &InitializeParticle(
    Color=<rgba 1 0.8084 0.1796 1>
    Lifetime=19
    Mesh Scale Mode=::NewEnumerator3
    Mesh Uniform Scale=&FN_FrontEndFloatModifier(
     Gameplay=&RandomRangeFloat(Minimum=1, Maximum=4)
     Front End=&RandomRangeFloat(Minimum=0.05, Maximum=0.1)
    )
    Position Offset=&AddVectorToPosition(
     Position=&MakeVector(X=0, Y=0, Z=&Multiply_Float(A=-4000, B=$User.Height))
     Vector=<0 0 0>
    )
    Sprite Size Mode=::NewEnumerator3
    Uniform Sprite Size=1
    UsePositionOffset=true
    Write Material Random=true
   )
   &ShapeLocation(
    Cylinder Height=300
    Cylinder Radius=4000
    MaskSpawnGroup=false
    Override Seed=false
    Shape Primitive=::NewEnumerator1
    Surface Only=false
    UseOwnerScale=true
   )
   &InitialMeshOrientation(Use Rotation Vector=false)
   &AddVelocity(Velocity=<0 0 1>, Velocity Speed Scale=&RandomRangeFloat(Minimum=80, Maximum=160))
   &AddRotationalVelocity(Rotation Rate=&RandomRangeVector(Minimum=<-1 -1 -1>, Maximum=<1 1 1>))
  ]
  particleUpdateScript [
   &ParticleState()
   &PointAttractionForce(
    Attraction Radius=5000
    Attraction Strength=30
    Attractor Position Offset:Vector3f=<0 0 -700>
    Falloff Exponent=0
    Kill Within Radius=false
    Module.Use Falloff=true
    UseAttractorOffset=true
   )
   &UpdateMeshOrientation(Rotation Rate=&RandomRangeFloat(Minimum=0.0001, Maximum=-0.0002), Rotation Vector=<1 1 1>)
   &VortexForce(UseOriginOffset=true, Vortex Force Amount=&RandomRangeFloat(Minimum=1, Maximum=3))
   &ScaleMeshSize(Scale Factor=&Multiply_VectorByFloat(
     Vector=<1 1 1>
     Float=&FloatFromCurve(
      FloatCurve=DataInterface() {propertyValues {Curve {keys [
          {arriveTangent=5.7795, interpMode=RCIM_Cubic, leaveTangent=5.7795, tangentMode=RCTM_User, time=0, value=0}
          {arriveTangent=0.0388, leaveTangent=1, tangentMode=RCTM_User, time=0.1, value=1}
          <at 0.9 1>
          <at 1 0>
         ]}}}
      Scale Curve=&RandomRangeFloat(
       Minimum=&Lerp_Float(
        A=0.5
        B=1
        Alpha=&ScalabilityDistanceBasedFloat(
         Curve:NiagaraDataInterfaceCurve=DataInterface() {propertyValues {Curve {keys [
             {arriveTangent=5.7795, interpMode=RCIM_Cubic, leaveTangent=5.7795, tangentMode=RCTM_User, time=0, value=0}
             {arriveTangent=0.0388, leaveTangent=1, tangentMode=RCTM_User, time=0.1, value=1}
             <at 0.9 1>
             <at 1 0>
            ]}}}
         CurveIndex:NiagaraFloat=$Emitter.Scalability.DistanceFraction
         Bias:NiagaraFloat=0
         Scale:NiagaraFloat=1
        )
       )
       Maximum=&Lerp_Float(
        A=0.7
        B=3
        Alpha=&ScalabilityDistanceBasedFloat(
         Curve:NiagaraDataInterfaceCurve=DataInterface() {propertyValues {Curve {keys [
             {arriveTangent=6.3036, interpMode=RCIM_Cubic, leaveTangent=6.3036, tangentMode=RCTM_User, time=0, value=0}
             {arriveTangent=-0.0137, leaveTangent=1, tangentMode=RCTM_User, time=0.1, value=1}
             <at 0.9 1>
             <at 1 0>
            ]}}}
         CurveIndex:NiagaraFloat=$Emitter.Scalability.DistanceFraction
         Bias:NiagaraFloat=0
         Scale:NiagaraFloat=1
        )
       )
      )
     )
    ))
   /SetVariables(Particles.MeshIndex=&RandomRangeInt(Minimum=0, Maximum=3)) {isSetParameters=true}
   &Drag(
    Drag=&FloatFromCurve(FloatCurve=DataInterface() {propertyValues {Curve {keys [<at 0 0.05 RCIM_Cubic>,<at 1 0 RCIM_Cubic>]}}})
    Rotational Drag:NiagaraFloat=&FloatFromCurve(FloatCurve=DataInterface() {propertyValues {Curve {
        defaultValue=0
        keys [{interpMode=RCIM_Cubic, tangentMode=RCTM_User, time=0, value=0.2}, {interpMode=RCIM_Cubic, tangentMode=RCTM_User, time=1, value=0.4}]
       }}})
    Use Linear Drag=true
    Use Rotational Drag=true
   )
   &SolveForcesAndVelocity(Clamp Velocity=false, Limit Acceleration=false)
   /SetVariables(
    Particles.Glow Camera Offset=30
    Particles.Glow Color=<rgba 1 1 1 200>
    Particles.Glow Mesh Scale:Vector3f=&Multiply_VectorByFloat(Vector=$Particles.Scale, Float=0.3)
   ) {isSetParameters=true}
  ]
  rendererClasses [&NiagaraMeshRendererProperties]
  renderers [
   {rendererClass=&NiagaraMeshRendererProperties, rendererIndex=0}
   {rendererClass=&NiagaraMeshRendererProperties, rendererIndex=1}
  ]
 }
 Impact_Flare {
  simTarget=GPUComputeSim
  emitterUpdateScript [
   &EmitterState(
    Enable Distance Culling=true
    Life Cycle Mode=::NewEnumerator1
    Loop Duration=1
    MaxDistance=100000
    Min Distance Response=::SleepAndLetParticlesFinish
    MinDistance=2000
    Module.bUseMaxDistance=true
    Module.bUseMinDistance=true
    Scalability Mode=::NewEnumerator1
    UseLoopDelay=false
   )
   &SpawnBurst_Instantaneous(Use Spawn Probability=false, UseLoopCountLimit=false)
   &SpawnBurst_Instantaneous(Spawn Group=1, Use Spawn Probability=false, UseLoopCountLimit=false) {module=SpawnBurst_Instantaneous001}
   &SpawnRate(SpawnGroup=2, Use Spawn Probability=false)
   &SpawnBurst_Instantaneous(
    Spawn Count=&RandomRangeInt(Minimum=1, Maximum=1)
    Spawn Group=3
    Use Spawn Probability=false
    UseLoopCountLimit=false
   ) {enabled=false, module=SpawnBurst_Instantaneous002}
  ]
  particleSpawnScript [
   /SetVariables(Particles.VisibilityTag=&SpawnGroup()) {isSetParameters=true, module=SetVariables_#9}
   &InitializeParticle(
    Lifetime=&MaskFloatBySpawnGroup(
     Masked Float=2
     Passthrough Float=&MaskFloatBySpawnGroup(
      Masked Float=5
      Passthrough Float=&MaskFloatBySpawnGroup(Masked Float=5, Passthrough Float=&RandomRangeFloat(Minimum=1, Maximum=2), Spawn Group Mask=2)
      Spawn Group Mask=1
     )
    )
    Mesh Scale=<12.5 12.5 12.5>
    Mesh Scale Mode=::NewEnumerator4
    Sprite Rotation Angle=&MaskFloatBySpawnGroup(Masked Float=0, Passthrough Float=&RandomRangeFloat(Minimum=-360, Maximum=360))
    Sprite Rotation Mode=::NewEnumerator0
    Sprite Size Mode=::NewEnumerator3
    Uniform Sprite Size=10000
    UsePositionOffset=false
    Write Material Random=true
   )
   &DynamicMaterialParameters(
    Index 0 Param 1=1
    Index 0 Param 2=1
    Index 0 Param 3=1
    Index 0 Param 4=1
    Module.Index0Float0Write=true
    Module.Index0Float1Write=true
    Module.Index0Float2Write=true
    Module.Index0Float3Write=true
   )
   /SetVariables() {isSetParameters=true}
   &InitialMeshOrientation(
    Mesh Axis To Orient=&RandomRangeVector(Minimum=<0 0 1>, Maximum=<0 0 -1>)
    Mesh Orientation Mode=::NewEnumerator2
    Use Rotation Vector=false
   )
  ]
  particleUpdateScript [
   &ParticleState()
   &ScaleSpriteSize(
    Uniform Curve Scale=&Waveform("Waveform [1]"=::NewEnumerator12, "[1] Frequency"=12, "[1] Amplitude Min/Max"=<0.9 1>)
    Uniform Curve Sprite Scale=DataInterface() {propertyValues {Curve {
       defaultValue=0
       keys [
        {interpMode=RCIM_Cubic, tangentMode=RCTM_User, time=0, value=0.8}
        {interpMode=RCIM_Cubic, tangentMode=RCTM_User, time=0.5, value=1}
        {tangentMode=RCTM_User, time=1, value=0.8}
       ]
      }}}
   )
   &ScaleSpriteSize(
    Scale Sprite Size Mode=::NewEnumerator0
    Uniform Scale Factor=&MakeFloatFromVector(
     Vector=&VectorFromFloat(Value=&Lerp_Float(
       A=0
       B=1
       Alpha=&ScalabilityDistanceBasedFloat(
        Curve:NiagaraDataInterfaceCurve=DataInterface() {propertyValues {Curve {defaultValue=0, keys [{tangentMode=RCTM_User, time=0, value=0}, <at 0.05 1>, {tangentMode=RCTM_User, time=1, value=1}]}}}
        CurveIndex:NiagaraFloat=$Emitter.Scalability.DistanceFraction
        Bias:NiagaraFloat=0
        Scale:NiagaraFloat=1
       )
      ))
     Channel=::NewEnumerator0
    )
   ) {module=ScaleSpriteSize001}
   &ScaleMeshSize(Scale Factor=&VectorFromFloat(Value=&Lerp_Float(
      A=0
      B=1
      Alpha=&ScalabilityDistanceBasedFloat(
       Curve:NiagaraDataInterfaceCurve=DataInterface() {propertyValues {Curve {defaultValue=0, keys [{tangentMode=RCTM_User, time=0, value=0}, <at 0.05 1>, {tangentMode=RCTM_User, time=1, value=1}]}}}
       CurveIndex:NiagaraFloat=$Emitter.Scalability.DistanceFraction
       Bias:NiagaraFloat=0
       Scale:NiagaraFloat=1
      )
     )))
   &CurlNoiseForce(
    Noise Frequency=1
    Noise Strength=&MaskFloatBySpawnGroup(
     Masked Float=0
     Passthrough Float=&MaskFloatBySpawnGroup(
      Masked Float=0
      Passthrough Float=&MaskFloatBySpawnGroup(Masked Float=0, Passthrough Float=100, Spawn Group Mask=2)
      Spawn Group Mask=1
     )
    )
    Pan Noise=true
    Pan Noise Field=<0.1 0.05 -0.2>
   )
   &Drag(
    Drag=&MaskFloatBySpawnGroup(
     Masked Float=0
     Passthrough Float=&MaskFloatBySpawnGroup(
      Masked Float=0
      Passthrough Float=&MaskFloatBySpawnGroup(Masked Float=0, Passthrough Float=&RandomRangeFloat(Minimum=0.1, Maximum=0.2), Spawn Group Mask=2)
      Spawn Group Mask=1
     )
    )
    Use Linear Drag=true
    Use Rotational Drag=true
   )
   &Timeline(
    Play Rate=&FloatFromCurve(FloatCurve=DataInterface() {propertyValues {Curve {keys [<at 0 2>,<at 1 1.3>]}}})
    Timeline End=1000
   )
   /SetVariables(Particles.EnergyLeakColor:LinearColor=&Multiply_LinearColorByFloat(
     Linear Color=<rgba 1.2 0.1667 0 1>
     Float=&FloatFromCurve(
      FloatCurve=DataInterface() {propertyValues {Curve {keys [
          {arriveTangent=6.7513, interpMode=RCIM_Cubic, leaveTangent=6.7513, tangentMode=RCTM_User, time=0, value=0}
          {arriveTangent=0.1216, interpMode=RCIM_Cubic, leaveTangent=0.1216, tangentMode=RCTM_User, time=0.2, value=0.9847}
          {arriveTangent=-0.0116, interpMode=RCIM_Cubic, leaveTangent=-0.0116, tangentMode=RCTM_User, time=1, value=0}
         ]}}}
      Scale Curve=1.2
     )
    )) {enabled=false, isSetParameters=true, module=SetVariables_#10}
   /SetVariables(Particles.EnergyLeakSize:Vector2f=&Vector2DFromFloat(Value=&Multiply_Float(
      A=900
      B=&FloatFromCurve(
       FloatCurve=DataInterface() {propertyValues {Curve {keys [
           {interpMode=RCIM_Cubic, tangentMode=RCTM_Break, time=0, value=0.5}
           {arriveTangent=-0.3016, interpMode=RCIM_Cubic, tangentMode=RCTM_Break, time=0.75, value=1}
          ]}}}
       Scale Curve=&Lerp_Float(
        A=4
        B=4
        Alpha=&ScalabilityDistanceBasedFloat(
         Curve:NiagaraDataInterfaceCurve=DataInterface() {propertyValues {Curve {defaultValue=0, keys [{tangentMode=RCTM_User, time=0, value=0}, <at 0.05 1>, {tangentMode=RCTM_User, time=1, value=1}]}}}
         CurveIndex:NiagaraFloat=$Emitter.Scalability.DistanceFraction
         Bias:NiagaraFloat=0
         Scale:NiagaraFloat=1
        )
       )
      )
     ))) {enabled=false, isSetParameters=true, module=SetVariables_#11}
   /SetVariables(Particles.EnergyLeak_Dyanmic_01:Vector4f=&MakeVector4(X=&Multiply_Float(A=$Particles.Timeline.Time, B=0.5), Y=0, Z=0, W=0)) {enabled=false, isSetParameters=true, module=SetVariables_#12}
   /SetVariables(Particles.Plasma_Dynamic_01:Vector4f=&MakeVector4(
     X=&FloatFromCurve(FloatCurve=DataInterface() {propertyValues {Curve {keys [<at 0 0.2 RCIM_Cubic>,<at 0.5 0>]}}})
     Y=&FloatFromCurve(FloatCurve=DataInterface() {propertyValues {Curve {keys [<at 0 1>,<at 0.5 0.75>]}}})
     Z=&FloatFromCurve(FloatCurve=DataInterface() {propertyValues {Curve {keys [<at 0 0>,<at 1 1>,<at 1 1>]}}})
     W=0
    )) {isSetParameters=true}
   /SetVariables(Particles.Plasma_COlor:LinearColor=&ColorFromCurve(ColorCurve=DataInterface() {propertyValues {
       AlphaCurve {keys [<at 0 0>,<at 0.1582 0.1>,<at 0.9906 0>]}
       BlueCurve {keys [<at 0 0.0334>,<at 0.9825 0.0065>]}
       GreenCurve {keys [<at 0 0.2055>,<at 0.9825 0.7547>]}
       RedCurve {keys [<at 0 1>,<at 0.9825 2>]}
      }})) {isSetParameters=true}
   &CameraOffset(Camera Offset Amount=&Multiply_Float(
     A=&FN_CameraDistance(
      0Dist:NiagaraFloat=0
      1Dist:NiagaraFloat=800
      Camera Query:NiagaraDataInterfaceCamera=DataInterface() {propertyValues {bRequireCurrentFrameData=false}}
      Position:NiagaraPosition=$Particles.Position
     )
     B=1000
    ))
   &ScaleColor(
    Scale Alpha=&FloatFromCurve(FloatCurve=DataInterface() {propertyValues {Curve {keys [
         {arriveTangent=-0.1319, interpMode=RCIM_Cubic, leaveTangent=-0.1319, tangentMode=RCTM_User, time=0, value=0}
         <at 0.5 1 RCIM_Cubic>
         <at 1 0 RCIM_Cubic>
        ]}}})
    Scale RGB=&MakeVectorFromLinearColorRGB(Color=&ColorFromCurve(ColorCurve=DataInterface() {propertyValues {
        AlphaCurve {keys []}
        BlueCurve {keys [<at 0 1.1082 RCIM_Cubic>,<at 0.3877 0.0275>,<at 1 0.025 RCIM_Cubic>]}
        GreenCurve {keys [<at 0 1.7837 RCIM_Cubic>,<at 0.3877 0.6197>,<at 1 0 RCIM_Cubic>]}
        RedCurve {keys [<at 0 2 RCIM_Cubic>,<at 0.3877 1>,<at 1 0.1878 RCIM_Cubic>]}
       }}))
    ScaleA=true
    ScaleRGB=true
   )
   /SetVariables(Particles.Position=&AddVectorToPosition(Position=<0 0 0>, Vector=$System.OrbWPOLocal)) {isSetParameters=true, module=SetVariables_#13}
   &FN_Sprite_Attributes(
    Color=&Multiply_LinearColorByFloat(Linear Color=$Particles.Color, Float=2)
    Create Color Attribute=true
    Create SpriteSize Attribute=true
    SpriteSize=&Multiply_Vector2D(A=$Particles.SpriteSize, B=<1 0.4232>)
    Write to Module Namespace=true
   )
   &SolveForcesAndVelocity(Clamp Velocity=false, Limit Acceleration=false)
  ]
  rendererClasses [&NiagaraSpriteRendererProperties, &NiagaraMeshRendererProperties]
  renderers [
   {rendererClass=&NiagaraSpriteRendererProperties}
   {rendererClass=&NiagaraSpriteRendererProperties, rendererIndex=1}
   {rendererClass=&NiagaraMeshRendererProperties, rendererIndex=2}
  ]
 }
 Large_Cloud {
  bEnabled=false
  emitterUpdateScript [
   &EmitterState(
    Enable Distance Culling=true
    Life Cycle Mode=::NewEnumerator1
    MaxDistance=100000
    Min Distance Response=::SleepAndLetParticlesFinish
    MinDistance=5000
    Module.bApplySpawnCountScaleByDistance=true
    Module.bUseMaxDistance=true
    Module.bUseMinDistance=true
    Scalability Mode=::NewEnumerator1
    Scale Spawn Count=true
    Spawn Count Scale By Distance=DataInterface() {propertyValues {Curve {keys [<at 0 0>,<at 0.3 2>,<at 1 2>]}}}
    UseLoopDelay=false
   )
   &SpawnBurst_Instantaneous(Spawn Count=20, Use Spawn Probability=false, UseLoopCountLimit=false) {enabled=false}
   &SpawnRate(SpawnRate=2, Use Spawn Probability=false)
  ]
  particleSpawnScript [
   &InitializeParticle(
    Color=<rgba 1 0.3641 0.0922 0.7043>
    Lifetime=&RandomRangeFloat(Minimum=3, Maximum=5)
    Sprite Rotation Mode=::NewEnumerator2
    Sprite Size Mode=::NewEnumerator1
    Sprite UV Mode=::NewEnumerator3
    Uniform Sprite Size Max=5000
    Uniform Sprite Size Min=4000
    UsePositionOffset=false
    Write Material Random=true
   )
   &ShapeLocation(MaskSpawnGroup=false, Override Seed=false, Ring Radius=1500, Shape Primitive=::NewEnumerator4, UseOwnerScale=true)
   &AddVelocity(Velocity=&RandomRangeVector(Minimum=<-10 -10 300>, Maximum=<10 10 400>)) {enabled=false, module=AddVelocity001}
  ]
  particleUpdateScript [
   &ParticleState()
   &SpriteRotationRate(Rotation Rate=&RandomRangeFloat(Minimum=-3, Maximum=5))
   &SubUVAnimation(
    Evaluation Type=::NewEnumerator0
    "Sprite Renderer "=DataInterface() {propertyValues {SpriteRenderer=&self:Large_Cloud_0.NiagaraSpriteRendererProperties_0}}
    SubUV Animation Mode:ENiagara_SubUVLookupModeV2=::NewEnumerator2
    UseEndFrame=false
    UseStartFrame=false
   )
   &ScaleSpriteSize(Uniform Curve Sprite Scale=DataInterface() {propertyValues {Curve {keys [{tangentMode=RCTM_User, time=0, value=0.85}, {tangentMode=RCTM_User, time=1, value=1}]}}}) {module=ScaleSpriteSize001}
   &ScaleSpriteSize(
    Scale Factor=&Vector2DFromFloat(Value=&Lerp_Float(
      A=0
      B=1
      Alpha=&ScalabilityDistanceBasedFloat(
       Curve:NiagaraDataInterfaceCurve=DataInterface() {propertyValues {Curve {defaultValue=0.1, keys [{tangentMode=RCTM_User, time=0, value=0.6}, {tangentMode=RCTM_User, time=1, value=2.2}]}}}
       CurveIndex:NiagaraFloat=$Emitter.Scalability.DistanceFraction
       Bias:NiagaraFloat=0
       Scale:NiagaraFloat=1
      )
     ))
    Scale Sprite Size Mode=::NewEnumerator2
   )
   &DynamicMaterialParameters(
    Module.Index0Float0Write=true
    Module.Index0Float1Write=true
    Module.Index0Float2Write=true
    Module.Index0Float3Write=true
    Module.Index1Float0Write=true
    Module.Index1Float1Write=true
    Module.Index1Float2Write=true
    Module.Index1Float3Write=true
    Write Parameter Index 1=::NewEnumerator1
   )
   &ScaleColor(
    Scale Alpha=&FloatFromCurve(
     FloatCurve=DataInterface() {propertyValues {Curve {
        defaultValue=1
        keys [
         {arriveTangent=0.0123, interpMode=RCIM_Cubic, leaveTangent=0.0123, tangentMode=RCTM_User, time=0, value=0}
         {arriveTangent=-0.0119, interpMode=RCIM_Cubic, leaveTangent=-0.0119, tangentMode=RCTM_User, time=0.5, value=0.35}
         {arriveTangent=-0.2138, interpMode=RCIM_Cubic, leaveTangent=-0.2138, tangentMode=RCTM_User, time=1, value=0}
        ]
       }}}
     Scale Curve=&Multiply_FloatByInt(Float=$User.Alpha, Integer=1)
    )
    ScaleA=true
    ScaleRGB=true
   )
   &ScaleColor(
    Scale Alpha=&Lerp_Float(
     A=0
     B=1
     Alpha=&ScalabilityDistanceBasedFloat(
      Curve:NiagaraDataInterfaceCurve=DataInterface() {propertyValues {Curve {defaultValue=1, keys [{tangentMode=RCTM_User, time=0, value=0.1}, <at 0.35 0.5>, {tangentMode=RCTM_User, time=1, value=0.5}]}}}
      CurveIndex:NiagaraFloat=$Emitter.Scalability.DistanceFraction
      Bias:NiagaraFloat=0
      Scale:NiagaraFloat=1
     )
    )
    ScaleA=true
    ScaleRGB=true
   ) {module=ScaleColor001}
   &SolveForcesAndVelocity(Clamp Velocity=false, Limit Acceleration=false)
   &CameraOffset(Camera Offset Amount=1500)
  ]
  rendererClasses [&NiagaraSpriteRendererProperties]
  renderers [{rendererClass=&NiagaraSpriteRendererProperties}]
 }
 Rays_Cone {
  emitterUpdateScript [
   &EmitterState(
    Enable Distance Culling=true
    Inactive Response=::NewEnumerator1
    Life Cycle Mode=::NewEnumerator1
    Loop Duration=0.25
    MaxDistance=100000
    Min Distance Response=::SleepAndLetParticlesFinish
    MinDistance=4000
    Module.bApplySpawnCountScaleByDistance=true
    Module.bUseMaxDistance=true
    Module.bUseMinDistance=true
    Scalability Mode=::NewEnumerator1
    Scale Spawn Count=true
    Spawn Count Scale By Distance=DataInterface() {propertyValues {Curve {keys [<at 0 0.2>,<at 0.2 2>,<at 1 2>]}}}
    UseLoopDelay=true
   )
   &SpawnBurst_Instantaneous(
    Spawn Count=&RandomRangeInt(Minimum=1, Maximum=2)
    Use Spawn Probability=false
    UseLoopCountLimit=false
   )
  ]
  particleSpawnScript [
   &InitializeParticle(
    Color=<rgba 300 178.5462 140.625 0.3913>
    Lifetime=&RandomRangeFloat(Minimum=3, Maximum=6)
    Position Offset=<0 0 -1000>
    Sprite Size=&RandomRangeVector2D(Minimum=<80 120>, Maximum=<100 400>)
    Sprite Size Mode=::NewEnumerator4
    Sprite UV Mode=::NewEnumerator4
    Sprite UV Scale:Vector2f=<1 -1>
    UsePositionOffset=false
    Write Material Random=true
   )
   &DynamicMaterialParameters(
    Index 0 Param 1=&RandomRangeFloat(Minimum=0.8, Maximum=2)
    Index 0 Param 2=&RandomRangeFloat(Minimum=0.5, Maximum=1.5)
    Index 0 Param 3=&RandomRangeFloat(Minimum=0.6, Maximum=1.6)
    Index 0 Param 4=&RandomRangeFloat(Minimum=0.5, Maximum=1.5)
    Module.Index0Float0Write=true
    Module.Index0Float1Write=true
    Module.Index0Float2Write=true
    Module.Index0Float3Write=true
   )
  ]
  particleUpdateScript [
   &ParticleState()
   &PointAttractionForce(
    Attraction Radius=15000
    Attraction Strength=100
    Attractor Position Offset:Vector3f=<0 0 -200>
    Falloff Exponent=0
    Kill Within Radius=false
    Module.Use Falloff=true
    UseAttractorOffset=true
   )
   &ScaleSpriteSize(
    Scale Factor=&MakeVector2D(
     X=1
     Y=&FloatFromCurve(
      FloatCurve=DataInterface() {propertyValues {Curve {defaultValue=-2, keys [{tangentMode=RCTM_User, time=0, value=0}, {tangentMode=RCTM_User, time=1, value=1.5}]}}}
      Scale Curve=&Multiply_Float(A=&RandomRangeFloat(Minimum=20, Maximum=25), B=$User.Height)
     )
    )
    Scale Sprite Size Mode=::NewEnumerator2
   )
   /SetVariables(Particles.Position=&AddVectorToPosition(
     Position=&RandomRangeVector(Minimum=<-4000 -4000 -4000>, Maximum=<4000 4000 -5000>)
     Vector=$System.OrbWPOLocal
    )) {isSetParameters=true}
   &CameraOffset(Camera Offset Amount=25)
   &ScaleColor(
    Linear Color Curve=DataInterface() {propertyValues {
      AlphaCurve {keys [<at 0 0>,<at 0.1796 0.1>,<at 0.4976 0>]}
      BlueCurve {keys [<at 0 1>,<at 1 0.1483>]}
      GreenCurve {keys [<at 0 1>,<at 1 0.2501>]}
      RedCurve {keys [<at 0 1>,<at 1 0.3229>]}
     }}
    Scale Mode=::NewEnumerator2
   )
   &SolveForcesAndVelocity(Clamp Velocity=false, Limit Acceleration=false)
  ]
  rendererClasses [&NiagaraSpriteRendererProperties]
  renderers [{rendererClass=&NiagaraSpriteRendererProperties}]
 }
 Sprites_Dot {
  emitterUpdateScript [
   &EmitterState(
    Enable Distance Culling=true
    Life Cycle Mode=::NewEnumerator1
    Loop Duration=8
    MaxDistance=100000
    Min Distance Response=::SleepAndLetParticlesFinish
    MinDistance=3000
    Module.bApplySpawnCountScaleByDistance=true
    Module.bUseMaxDistance=true
    Module.bUseMinDistance=true
    Scalability Mode=::NewEnumerator1
    Scale Spawn Count=true
    Spawn Count Scale By Distance=DataInterface() {propertyValues {Curve {keys [<at 0 0>,<at 0.2 1>,<at 1 1>]}}}
    UseLoopDelay=false
   )
   &SpawnRate(SpawnRate=10, Use Spawn Probability=false)
  ]
  particleSpawnScript [
   &InitializeParticle(
    Color=<rgba 10 6.9864 5.3125 1>
    Lifetime=8
    Mass Mode=::NewEnumerator3
    Position Offset=<0 0 -3800>
    Sprite Rotation Mode=::NewEnumerator2
    Sprite Size Mode=::NewEnumerator1
    Uniform Sprite Size Max=100
    Uniform Sprite Size Min=70
    UsePositionOffset=true
    Write Material Random=true
   )
   &AddVelocity(
    Cone Angle=1
    Cone Axis=<0 0 1>
    Module.UseConeFalloff=false
    Module.UseVelDistribution=false
    Override Seed=false
    Velocity Mode=::NewEnumerator2
    Velocity Speed=&RandomRangeFloat(
     Minimum=&Multiply_Float(A=1, B=$User.Height)
     Maximum=&Multiply_Float(A=2, B=$User.Height)
    )
   ) {enabled=false}
   &ShapeLocation(
    Cylinder Height=1000
    Cylinder Radius=3000
    MaskSpawnGroup=false
    Override Seed=false
    Shape Primitive=::NewEnumerator1
    Surface Only=false
    UseOwnerScale=true
   ) {module=ShapeLocation001}
   &ShapeLocation(
    Box Midpoint=<0.5 0.5 1>
    Box Size=&MakeVector(X=80, Y=80, Z=&Multiply_Float(A=4000, B=$User.Height))
    MaskSpawnGroup=false
    Override Seed=false
    Shape Primitive=::NewEnumerator2
    Surface Only=false
    UseOwnerScale=true
   ) {enabled=false}
  ]
  particleUpdateScript [
   &ParticleState()
   &PointAttractionForce(
    Attraction Radius=5000
    Attraction Strength=&RandomRangeFloat(Minimum=20, Maximum=40)
    Falloff Exponent=0
    Kill Within Radius=false
    Module.Use Falloff=true
    UseAttractorOffset=false
   )
   &PointAttractionForce(
    Attraction Radius=7000
    Attraction Strength=&RandomRangeFloat(Minimum=20, Maximum=40)
    Attractor Position Offset:Vector3f=<0 0 -2000>
    Falloff Exponent=0
    Kill Within Radius=false
    Module.Use Falloff=true
    UseAttractorOffset=true
   ) {module=PointAttractionForce001}
   &CurlNoiseForce(Noise Frequency=25, Noise Strength=30, Pan Noise=true, Pan Noise Field=<0.1 0.05 -0.2>)
   &Drag(Drag=&RandomRangeFloat(Minimum=0.05, Maximum=0.06), Use Linear Drag=true, Use Rotational Drag=false)
   &ScaleColor(
    Scale Alpha=&FloatFromCurve(
     FloatCurve=DataInterface() {propertyValues {Curve {keys [
         <at 0 0 RCIM_Cubic>
         {arriveTangent=1.1111, leaveTangent=1.1111, time=0.2, value=1}
         {arriveTangent=-1.25, interpMode=RCIM_Cubic, leaveTangent=-1.25, tangentMode=RCTM_Break, time=0.9, value=1}
         <at 1 0 RCIM_Cubic>
        ]}}}
     Scale Curve=&RandomRangeFloat(
      Minimum=&Lerp_Float(
       A=0.02
       B=0.6
       Alpha=&ScalabilityDistanceBasedFloat(
        Curve:NiagaraDataInterfaceCurve=DataInterface() {propertyValues {Curve {keys [<at 0 0 RCIM_Cubic>, {arriveTangent=1.1111, leaveTangent=1.1111, time=0.1, value=1}, <at 1 1 RCIM_Cubic>]}}}
        CurveIndex:NiagaraFloat=$Emitter.Scalability.DistanceFraction
        Bias:NiagaraFloat=0
        Scale:NiagaraFloat=0.8
       )
      )
      Maximum=&Lerp_Float(
       A=0.06
       B=1
       Alpha=&ScalabilityDistanceBasedFloat(
        Curve:NiagaraDataInterfaceCurve=DataInterface() {propertyValues {Curve {keys [<at 0 0 RCIM_Cubic>, {arriveTangent=1.1111, leaveTangent=1.1111, time=0.1, value=1}, <at 1 1 RCIM_Cubic>]}}}
        CurveIndex:NiagaraFloat=$Emitter.Scalability.DistanceFraction
        Bias:NiagaraFloat=0
        Scale:NiagaraFloat=1
       )
      )
     )
    )
    ScaleA=true
    ScaleRGB=true
   )
   &GravityForce(Gravity=&RandomRangeVector(Minimum=<0 0 30>, Maximum=<0 0 50>))
   &SolveForcesAndVelocity(Clamp Velocity=false, Limit Acceleration=false)
   &ScaleSpriteSize(Scale Factor=&Vector2DFromFloat(Value=&FloatFromCurve(FloatCurve=DataInterface() {propertyValues {Curve {keys [
          {interpMode=RCIM_Cubic, tangentMode=RCTM_User, time=0, value=0}
          {arriveTangent=1.1111, leaveTangent=1.1111, time=0.1, value=1}
          {arriveTangent=-1.1111, leaveTangent=-1.1111, time=0.9, value=1}
          <at 1 0 RCIM_Cubic>
         ]}}})))
   &ScaleSpriteSize(
    Scale Sprite Size Mode=::NewEnumerator0
    Uniform Scale Factor=&MakeFloatFromVector(
     Vector=&VectorFromFloat(Value=&Lerp_Float(
       A=0.15
       B=1
       Alpha=&ScalabilityDistanceBasedFloat(
        Curve:NiagaraDataInterfaceCurve=DataInterface() {propertyValues {Curve {defaultValue=0, keys [{tangentMode=RCTM_User, time=0, value=0}, <at 0.02 1>, {tangentMode=RCTM_User, time=1, value=1}]}}}
        CurveIndex:NiagaraFloat=$Emitter.Scalability.DistanceFraction
        Bias:NiagaraFloat=0
        Scale:NiagaraFloat=1
       )
      ))
     Channel=::NewEnumerator0
    )
   ) {module=ScaleSpriteSize001}
  ]
  rendererClasses [&NiagaraSpriteRendererProperties]
  renderers [{rendererClass=&NiagaraSpriteRendererProperties}]
 }
 Upward_Streaks {
  emitterUpdateScript [
   &EmitterState(
    Enable Distance Culling=true
    Life Cycle Mode=::NewEnumerator1
    Loop Delay=0.5
    Loop Duration=7
    MaxDistance=100000
    Min Distance Response=::SleepAndLetParticlesFinish
    MinDistance=4000
    Module.bApplySpawnCountScaleByDistance=true
    Module.bUseMaxDistance=true
    Module.bUseMinDistance=true
    Scalability Mode=::NewEnumerator1
    Scale Spawn Count=true
    Spawn Count Scale By Distance=DataInterface() {propertyValues {Curve {keys [<at 0 0>,<at 0.1 1>,<at 1 1>]}}}
    UseLoopDelay=false
   )
   &SpawnRate(SpawnRate=&RandomRangeFloat(Minimum=5, Maximum=10), Use Spawn Probability=false)
  ]
  particleSpawnScript [
   &InitializeParticle(
    Color=<rgba 1.2 1.0378 0.9 0.5304>
    Lifetime=&RandomRangeFloat(Minimum=3, Maximum=5)
    Sprite Size=&RandomRangeVector2D(Minimum=<50 3500>, Maximum=<80 5500>)
    Sprite Size Mode=::NewEnumerator4
    UsePositionOffset=false
    Write Material Random=true
   )
   &DynamicMaterialParameters(
    Index 0 Param 1=1
    Index 0 Param 2=4
    Index 0 Param 3=1
    Index 0 Param 4=1
    Module.Index0Float0Write=true
    Module.Index0Float1Write=true
    Module.Index0Float2Write=true
    Module.Index0Float3Write=true
   ) {enabled=false}
   &AddVelocity(Velocity=&RandomRangeVector(
     Minimum=&Multiply_VectorByFloat(Vector=<1 1 10>, Float=$User.Height)
     Maximum=&Multiply_VectorByFloat(Vector=<1 1 100>, Float=$User.Height)
    )) {enabled=false}
  ]
  particleUpdateScript [
   &ParticleState()
   &PointAttractionForce(
    Attraction Radius=7000
    Attraction Strength=100
    Attractor Position Offset:Vector3f=<0 0 1500>
    Falloff Exponent=0
    Kill Within Radius=false
    Module.Use Falloff=true
    UseAttractorOffset=true
   )
   /SetVariables(Particles.Position=&AddVectorToPosition(
     Position=&RandomRangeVector(
      Minimum=&MakeVector(X=-600, Y=-600, Z=&Multiply_Float(A=-500, B=$User.Height))
      Maximum=&MakeVector(X=600, Y=600, Z=&Multiply_Float(A=-1000, B=$User.Height))
     )
     Vector=$System.OrbWPOLocal
    )) {isSetParameters=true}
   &CameraOffset(Camera Offset Amount=27.1877)
   &ScaleSpriteSize(
    Scale Factor=&Vector2DFromCurve(
     Vector2Curve=DataInterface() {propertyValues {
       XCurve {keys [<at 0 2.15 RCIM_Cubic>,<at 1 0 RCIM_Cubic>]}
       YCurve {keys [{arriveTangent=-2.543, interpMode=RCIM_Cubic, leaveTangent=-2.543, tangentMode=RCTM_User, time=0, value=1}, <at 1 -0.05 RCIM_Cubic>]}
      }}
     ScaleCurve=&Multiply_Vector2D_ByFloat(Float=$User.Height, Vector2D=<0.65 1>)
    )
    Scale Sprite Size Mode=::NewEnumerator2
   )
   &ScaleColor(
    Scale Alpha=&FloatFromCurve(
     FloatCurve=DataInterface() {propertyValues {Curve {
        defaultValue=1
        keys [<at 0 0 RCIM_Cubic>, {interpMode=RCIM_Cubic, tangentMode=RCTM_User, time=0.5, value=1}, {tangentMode=RCTM_User, time=1, value=0}]
       }}}
     Scale Curve=0.7
    )
    ScaleA=true
    ScaleRGB=true
   )
   &SolveForcesAndVelocity(Clamp Velocity=false, Limit Acceleration=false)
  ]
  rendererClasses [&NiagaraSpriteRendererProperties]
  renderers [{rendererClass=&NiagaraSpriteRendererProperties}]
 }
}
summary {
 systemName=NS_Orb_Tractor_Beam
 emitters [
  {emitterName=CoreBleed_Flare, rendererClasses [&NiagaraMeshRendererProperties]}
  {
   emitterName=Impact_Flare
   simTarget=GPUComputeSim
   rendererClasses [&NiagaraSpriteRendererProperties, &NiagaraMeshRendererProperties]
  }
  {emitterName=DarkSPline001, rendererClasses [&NiagaraMeshRendererProperties]}
  {emitterName=Sprites_Dot, rendererClasses [&NiagaraSpriteRendererProperties]}
  {emitterName=Floating_Rocks, rendererClasses [&NiagaraMeshRendererProperties]}
  {emitterName=Rays_Cone, rendererClasses [&NiagaraSpriteRendererProperties]}
  {emitterName=Upward_Streaks, rendererClasses [&NiagaraSpriteRendererProperties]}
  {emitterName=DarkSPline, rendererClasses [&NiagaraMeshRendererProperties]}
  {bEnabled=false, emitterName=Large_Cloud, rendererClasses [&NiagaraSpriteRendererProperties]}
  {emitterName=Beam_Sides, rendererClasses [&NiagaraMeshRendererProperties]}
  {emitterName=Cylinder004, rendererClasses [&NiagaraMeshRendererProperties, &NiagaraLightRendererProperties]}
  {emitterName=EnergyLeak, simTarget=GPUComputeSim, rendererClasses [&NiagaraSpriteRendererProperties]}
 ]
 userVariables [
  User.Alpha {type=NiagaraFloat, defaultValue=1, description=""}
  User.bIsFrontEnd {type=NiagaraBool}
  User.bIsFrontEndPreview {type=NiagaraBool}
  User.Height {type=NiagaraFloat, defaultValue=1, description=""}
  User.SmallBeam {type=NiagaraBool}
 ]
}
